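# Control Rig for the F_MED_PandaSpeak female body that procedurally corrects the outfit's apron, tie and pen bones.
# Weighted parent constraints seat the pen and tie, while spherical pose readers on spine and thighs drive tie offsets or store leg-pose alphas.
# Sequenced ModifyTransforms layers blend apron corrections from those alphas, gated by cape, crouch, slide, melee and emote states.
# Crouch and slide weights also use pelvis height, smoothed with AlphaInterp.
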ControlRigBlueprint /BRCosmetics/Characters/Player/Female/Medium/Bodies/F_MED_PandaSpeak/Meshes/F_MED_PandaSpeak_CtrlRig
graphs [
 Maximum_3_ContainedGraph {nodes [
   n1 = RigVMFunctionEntryNode()
   RigVMFunctionReturnNode(Result=&RigVMFunction_MathDoubleMax(B=@n1.C, A=&RigVMFunction_MathDoubleMax(A=@n1.A, B=@n1.B).Result).Result)
  ]}
 Maximum_3_ContainedGraph {nodes [
   n1 = RigVMFunctionEntryNode()
   RigVMFunctionReturnNode(Result=&RigVMFunction_MathDoubleMax(B=@n1.C, A=&RigVMFunction_MathDoubleMax(A=@n1.A, B=@n1.B).Result).Result)
  ]}
 Multiply_4_ContainedGraph {nodes [
   n1 = RigVMFunctionEntryNode()
   n2 = &RigVMFunction_MathDoubleMul(A=@n1.A, B=@n1.B)
   RigVMFunctionReturnNode(Result=&RigVMFunction_MathDoubleMul(B=@n1.D, A=&RigVMFunction_MathDoubleMul(B=@n1.C, A=@n2.Result).Result).Result)
  ]}
 New Function_ContainedGraph {nodes [
   RigVMFunctionEntryNode()
   > RigVMFunction_Sequence_1 = RigVMAggregateNode()
   &RigUnit_ParentConstraint(
    AdvancedSettings=(RotationOrderForFilter=XZY)
    Child=(Type=Bone,Name="base_pen")
    Parents="((Item=(Type=Bone,Name=\"thigh_r\"),Weight=0.0003),(Item=(Type=Bone,Name=\"dyn_apron_fr_c_00\"),Weight=0.0004),(Item=(Type=Bone,Name=\"thigh_twist_01_l\"),Weight=0.0006),(Item=(Type=Bone,Name=\"dyn_apron_fr_r_01\"),Weight=0.0011),(Item=(Type=Bone,Name=\"dyn_apron_f ...(truncated)"
    Weight=1
    ExecutePin=@RigVMFunction_Sequence_1.A
   )
   > &RigUnit_ParentConstraint(
    AdvancedSettings=(RotationOrderForFilter=XZY)
    Child=(Type=Bone,Name="base_tie")
    Parents="((Item=(Type=Bone,Name=\"spine_04\"),Weight=0.0001),(Item=(Type=Bone,Name=\"thigh_twist_01_l\"),Weight=0.0029),(Item=(Type=Bone,Name=\"thigh_twist_01_r\"),Weight=0.0029),(Item=(Type=Bone,Name=\"spine_03\"),Weight=0.0076),(Item=(Type=Bone,Name=\"pelvis\"),Weight=0.18 ...(truncated)"
    Weight=1
   )
   > n1 = &RigUnit_SphericalPoseReader(
    ActiveRegionScaleFactors=(PositiveWidth=7,NegativeWidth=1,PositiveHeight=3,NegativeHeight=3)
    ActiveRegionSize=0.03
    Debug=(bDrawDebug=false,bDraw2D=False,bDrawLocalAxes=False,DebugScale=25,DebugSegments=20,DebugThickness=0.25)
    DriverAxis=(X=1,Y=0,Z=0)
    DriverItem=(Type=Bone,Name="spine_01")
    FalloffRegionScaleFactors=(PositiveWidth=1,NegativeWidth=1,PositiveHeight=1,NegativeHeight=1)
    FalloffSize=0.45
    FlipHeightScaling=False
    FlipWidthScaling=False
    OptionalParentItem=(Type=Bone,Name="None")
    RotationOffset=(X=-90,Y=0,Z=0)
   )
   n2 = &RigVMFunction_MathBoolNot(Value=RigVMVariableNode() {variable=isWearingCape}.Value)
   > &RigUnit_ModifyTransforms(
    ItemToModify=((Item=(Type=Bone,Name="base_tie"),Transform=(Rotation=(X=0,Y=0.5736,Z=0,W=0.8192),Translation=(X=-3,Y=0,Z=-1),Scale3D=(X=1,Y=1,Z=1))))
    WeightMinimum=0
    Weight=&RigVMFunction_MathDoubleMul(B=&RigVMFunction_MathBoolToFloat(Value=@n2.Result).Result, A=@n1.OutputParam).Result
   )
   > n3 = &RigUnit_SphericalPoseReader(
    ActiveRegionScaleFactors=(PositiveWidth=1,NegativeWidth=1,PositiveHeight=1,NegativeHeight=1)
    ActiveRegionSize=0.05
    Debug=(bDrawDebug=false,bDraw2D=False,bDrawLocalAxes=False,DebugScale=25,DebugSegments=20,DebugThickness=0.25)
    DriverAxis=(X=-1,Y=0,Z=0)
    DriverItem=(Type=Bone,Name="thigh_l")
    FalloffRegionScaleFactors=(PositiveWidth=1,NegativeWidth=1,PositiveHeight=0.9,NegativeHeight=0.9)
    FalloffSize=0.65
    FlipHeightScaling=False
    FlipWidthScaling=False
    OptionalParentItem=(Type=Bone,Name="None")
    RotationOffset=(X=-90,Y=0,Z=30)
   )
   > &RigUnit_ModifyTransforms(
    ItemToModify=((Item=(Type=Bone,Name="base_tie"),Transform=(Rotation=(X=0.1555,Y=0.1146,Z=0.1555,W=0.9688),Translation=(X=0,Y=0,Z=0),Scale3D=(X=1,Y=1,Z=1))))
    WeightMinimum=0
    Weight=@n3.OutputParam
   )
   > n4 = &RigUnit_SphericalPoseReader(
    ActiveRegionScaleFactors=(PositiveWidth=1,NegativeWidth=1,PositiveHeight=1,NegativeHeight=1)
    ActiveRegionSize=0.05
    Debug=(bDrawDebug=false,bDraw2D=False,bDrawLocalAxes=False,DebugScale=25,DebugSegments=20,DebugThickness=0.25)
    DriverAxis=(X=1,Y=0,Z=0)
    DriverItem=(Type=Bone,Name="thigh_r")
    FalloffRegionScaleFactors=(PositiveWidth=1,NegativeWidth=1,PositiveHeight=0.9,NegativeHeight=0.9)
    FalloffSize=0.65
    FlipHeightScaling=false
    FlipWidthScaling=false
    OptionalParentItem=(Type=Bone,Name="None")
    RotationOffset=(X=90,Y=0,Z=30)
   )
   > &RigUnit_ModifyTransforms(
    ItemToModify=((Item=(Type=Bone,Name="base_tie"),Transform=(Rotation=(X=-0.2273,Y=0.148,Z=-0.2273,W=0.9353),Translation=(X=0,Y=0,Z=0),Scale3D=(X=1,Y=1,Z=1))))
    WeightMinimum=0
    Weight=@n4.OutputParam
   )
   > &RigUnit_ModifyTransforms(
    ItemToModify=((Item=(Type=Bone,Name="base_tie"),Transform=(Rotation=(X=0,Y=0.0872,Z=0,W=0.9962),Translation=(X=0,Y=0,Z=0),Scale3D=(X=1,Y=1,Z=1))))
    WeightMinimum=0
    Weight=&RigVMFunction_MathBoolToFloat(Value=RigVMVariableNode() {variable=isParachuteOpen}.Value).Result
   )
   n5 = &RigVMFunction_MathBoolNot(Value=RigVMVariableNode() {variable=isWearingCape}.Value)
   n6 = RigVMVariableNode() {variable=isFalling}
   n7 = &RigVMFunction_AlphaInterp(
    InterpSpeedDecreasing=10
    InterpSpeedIncreasing=10
    Scale=1
    bClampResult=False
    Value=&RigVMFunction_MathBoolToFloat(Value=&RigVMFunction_MathBoolAnd(B=@n5.Result, A=@n6.Value).Result).Result
   )
   > &RigUnit_ModifyTransforms(
    ItemToModify=((Item=(Type=Bone,Name="base_tie"),Transform=(Rotation=(X=0,Y=0.342,Z=0,W=0.9397),Translation=(X=0,Y=0,Z=0),Scale3D=(X=1,Y=1,Z=1))))
    WeightMinimum=0
    Weight=@n7.Result
   )
   > &RigUnit_ModifyTransforms(
    ItemToModify="((Item=(Type=Bone,Name=\"base_tie\"),Transform=(Rotation=(X=0,Y=-0.1305,Z=0,W=0.9914),Translation=(X=0,Y=0,Z=0),Scale3D=(X=1,Y=1,Z=1))),(Item=(Type=Bone,Name=\"base_tie_l_loop\"),Transform=(Rotation=(X=-0.4881,Y=-0.1082,Z=0.1874,W=0.8455),Translation=(X=0,Y=0, ...(truncated)"
    WeightMinimum=0
    Weight=&RigVMFunction_MathBoolToFloat(Value=RigVMVariableNode() {variable=isWearingCape}.Value).Result
   )
   > &RigUnit_ModifyTransforms(
    ItemToModify=((Item=(Type=Bone,Name="base_tie"),Transform=(Rotation=(X=0,Y=0.0872,Z=0,W=0.9962),Translation=(X=0,Y=0,Z=0),Scale3D=(X=1,Y=1,Z=1))))
    WeightMinimum=0
    Weight=&RigVMFunction_MathBoolToFloat(Value=RigVMVariableNode() {variable=isFrontEnd}.Value).Result
   )
   n8 = &RigUnit_SphericalPoseReader(
    ActiveRegionScaleFactors=(PositiveWidth=1,NegativeWidth=1,PositiveHeight=1,NegativeHeight=1)
    ActiveRegionSize=0.05
    Debug=(bDrawDebug=false,bDraw2D=False,bDrawLocalAxes=False,DebugScale=25,DebugSegments=20,DebugThickness=0.25)
    DriverAxis=(X=-1,Y=0,Z=0)
    DriverItem=(Type=Bone,Name="thigh_l")
    FalloffRegionScaleFactors=(PositiveWidth=1,NegativeWidth=1,PositiveHeight=1,NegativeHeight=1)
    FalloffSize=0.66
    FlipHeightScaling=false
    FlipWidthScaling=false
    OptionalParentItem=(Type=Bone,Name="None")
    RotationOffset=(X=90,Y=0,Z=-30)
    ExecutePin=@RigVMFunction_Sequence_1.B
   )
   > RigVMVariableNode(Value=@n8.OutputParam) {variable=thigh_l_up_alpha}
   > n9 = &RigUnit_SphericalPoseReader(
    ActiveRegionScaleFactors=(PositiveWidth=1,NegativeWidth=1,PositiveHeight=1,NegativeHeight=1)
    ActiveRegionSize=0.05
    Debug=(bDrawDebug=false,bDraw2D=False,bDrawLocalAxes=False,DebugScale=25,DebugSegments=20,DebugThickness=0.25)
    DriverAxis=(X=1,Y=0,Z=0)
    DriverItem=(Type=Bone,Name="thigh_r")
    FalloffRegionScaleFactors=(PositiveWidth=1,NegativeWidth=1,PositiveHeight=1,NegativeHeight=1)
    FalloffSize=0.66
    FlipHeightScaling=false
    FlipWidthScaling=false
    OptionalParentItem=(Type=Bone,Name="None")
    RotationOffset=(X=-90,Y=0,Z=-30)
   )
   > RigVMVariableNode(Value=@n9.OutputParam) {variable=thigh_r_up_alpha}
   > n10 = &RigUnit_SphericalPoseReader(
    ActiveRegionScaleFactors=(PositiveWidth=1,NegativeWidth=1,PositiveHeight=1,NegativeHeight=1)
    ActiveRegionSize=0.02
    Debug=(bDrawDebug=false,bDraw2D=False,bDrawLocalAxes=False,DebugScale=25,DebugSegments=20,DebugThickness=0.25)
    DriverAxis=(X=-1,Y=0,Z=0)
    DriverItem=(Type=Bone,Name="thigh_l")
    FalloffRegionScaleFactors=(PositiveWidth=1,NegativeWidth=1.05,PositiveHeight=0.9,NegativeHeight=1)
    FalloffSize=0.63
    FlipHeightScaling=false
    FlipWidthScaling=false
    OptionalParentItem=(Type=Bone,Name="None")
    RotationOffset=(X=182,Y=30,Z=0)
   )
   > RigVMVariableNode(Value=@n10.OutputParam) {variable=thigh_l_side_alpha}
   > n11 = &RigUnit_SphericalPoseReader(
    ActiveRegionScaleFactors=(PositiveWidth=1,NegativeWidth=1,PositiveHeight=1,NegativeHeight=1)
    ActiveRegionSize=0.02
    Debug=(bDrawDebug=false,bDraw2D=False,bDrawLocalAxes=False,DebugScale=25,DebugSegments=20,DebugThickness=0.25)
    DriverAxis=(X=1,Y=0,Z=0)
    DriverItem=(Type=Bone,Name="thigh_r")
    FalloffRegionScaleFactors=(PositiveWidth=1.05,NegativeWidth=1,PositiveHeight=0.9,NegativeHeight=1)
    FalloffSize=0.63
    FlipHeightScaling=false
    FlipWidthScaling=false
    OptionalParentItem=(Type=Bone,Name="None")
    RotationOffset=(X=2,Y=30,Z=0)
   )
   > RigVMVariableNode(Value=@n11.OutputParam) {variable=thigh_r_side_alpha}
   > n12 = &RigUnit_SphericalPoseReader(
    ActiveRegionScaleFactors=(PositiveWidth=1,NegativeWidth=1,PositiveHeight=1,NegativeHeight=1)
    ActiveRegionSize=0.05
    Debug=(bDrawDebug=false,bDraw2D=False,bDrawLocalAxes=False,DebugScale=25,DebugSegments=20,DebugThickness=0.25)
    DriverAxis=(X=-1,Y=0,Z=0)
    DriverItem=(Type=Bone,Name="thigh_l")
    FalloffRegionScaleFactors=(PositiveWidth=1,NegativeWidth=1,PositiveHeight=0.9,NegativeHeight=0.9)
    FalloffSize=0.63
    FlipHeightScaling=false
    FlipWidthScaling=false
    OptionalParentItem=(Type=Bone,Name="None")
    RotationOffset=(X=140,Y=30,Z=0)
   )
   > RigVMVariableNode(Value=@n12.OutputParam) {variable=thigh_l_front_side_alpha}
   > n13 = &RigUnit_SphericalPoseReader(
    ActiveRegionScaleFactors=(PositiveWidth=1,NegativeWidth=1,PositiveHeight=1,NegativeHeight=1)
    ActiveRegionSize=0.05
    Debug=(bDrawDebug=false,bDraw2D=False,bDrawLocalAxes=False,DebugScale=25,DebugSegments=20,DebugThickness=0.25)
    DriverAxis=(X=1,Y=0,Z=0)
    DriverItem=(Type=Bone,Name="thigh_r")
    FalloffRegionScaleFactors=(PositiveWidth=1,NegativeWidth=1,PositiveHeight=0.9,NegativeHeight=0.9)
    FalloffSize=0.63
    FlipHeightScaling=false
    FlipWidthScaling=false
    OptionalParentItem=(Type=Bone,Name="None")
    RotationOffset=(X=-40,Y=30,Z=0)
   )
   > RigVMVariableNode(Value=@n13.OutputParam) {variable=thigh_r_front_side_alpha}
   > n14 = &RigUnit_SphericalPoseReader(
    ActiveRegionScaleFactors=(PositiveWidth=1,NegativeWidth=1,PositiveHeight=1.3,NegativeHeight=1)
    ActiveRegionSize=0.06
    Debug=(bDrawDebug=false,bDraw2D=False,bDrawLocalAxes=false,DebugScale=25,DebugSegments=20,DebugThickness=0.25)
    DriverAxis=(X=-1,Y=0,Z=0)
    DriverItem=(Type=Bone,Name="thigh_l")
    FalloffRegionScaleFactors=(PositiveWidth=1,NegativeWidth=1,PositiveHeight=1.3,NegativeHeight=1)
    FalloffSize=0.23
    FlipHeightScaling=False
    FlipWidthScaling=False
    OptionalParentItem=(Type=Bone,Name="None")
    RotationOffset=(X=55.4587,Y=-17.7371,Z=42.8005)
   )
   > RigVMVariableNode(Value=@n14.OutputParam) {variable=thigh_l_up_mid_alpha}
   > n15 = &RigUnit_SphericalPoseReader(
    ActiveRegionScaleFactors=(PositiveWidth=1,NegativeWidth=1,PositiveHeight=1.3,NegativeHeight=1)
    ActiveRegionSize=0.06
    Debug=(bDrawDebug=false,bDraw2D=False,bDrawLocalAxes=false,DebugScale=25,DebugSegments=20,DebugThickness=0.25)
    DriverAxis=(X=1,Y=0,Z=0)
    DriverItem=(Type=Bone,Name="thigh_r")
    FalloffRegionScaleFactors=(PositiveWidth=1,NegativeWidth=1,PositiveHeight=1.3,NegativeHeight=1)
    FalloffSize=0.23
    FlipHeightScaling=False
    FlipWidthScaling=False
    OptionalParentItem=(Type=Bone,Name="None")
    RotationOffset=(X=-124.5413,Y=-17.7371,Z=42.8005)
   )
   > RigVMVariableNode(Value=@n15.OutputParam) {variable=thigh_r_up_mid_alpha}
   > n16 = &RigUnit_SphericalPoseReader(
    ActiveRegionScaleFactors=(PositiveWidth=1,NegativeWidth=1.3,PositiveHeight=1,NegativeHeight=1)
    ActiveRegionSize=0.05
    Debug=(bDrawDebug=false,bDraw2D=False,bDrawLocalAxes=false,DebugScale=25,DebugSegments=20,DebugThickness=0.25)
    DriverAxis=(X=-1,Y=0,Z=0)
    DriverItem=(Type=Bone,Name="thigh_l")
    FalloffRegionScaleFactors=(PositiveWidth=1,NegativeWidth=1,PositiveHeight=1.3,NegativeHeight=1)
    FalloffSize=0.23
    FlipHeightScaling=False
    FlipWidthScaling=False
    OptionalParentItem=(Type=Bone,Name="None")
    RotationOffset=(X=59,Y=-24,Z=75)
   )
   > RigVMVariableNode(Value=@n16.OutputParam) {variable=thigh_l_down_mid_alpha}
   > n17 = &RigUnit_SphericalPoseReader(
    ActiveRegionScaleFactors=(PositiveWidth=1.3,NegativeWidth=1,PositiveHeight=1,NegativeHeight=1)
    ActiveRegionSize=0.05
    Debug=(bDrawDebug=false,bDraw2D=False,bDrawLocalAxes=false,DebugScale=25,DebugSegments=20,DebugThickness=0.25)
    DriverAxis=(X=1,Y=0,Z=0)
    DriverItem=(Type=Bone,Name="thigh_r")
    FalloffRegionScaleFactors=(PositiveWidth=1.3,NegativeWidth=1,PositiveHeight=1,NegativeHeight=1)
    FalloffSize=0.2
    FlipHeightScaling=False
    FlipWidthScaling=False
    OptionalParentItem=(Type=Bone,Name="None")
    RotationOffset=(X=-121,Y=-24,Z=75)
   )
   > RigVMVariableNode(Value=@n17.OutputParam) {variable=thigh_r_down_mid_alpha}
   > n18 = &RigUnit_SphericalPoseReader(
    ActiveRegionScaleFactors=(PositiveWidth=1,NegativeWidth=2,PositiveHeight=2,NegativeHeight=1)
    ActiveRegionSize=0.05
    Debug=(bDrawDebug=false,bDraw2D=False,bDrawLocalAxes=False,DebugScale=25,DebugSegments=20,DebugThickness=0.25)
    DriverAxis=(X=-1,Y=0,Z=0)
    DriverItem=(Type=Bone,Name="thigh_l")
    FalloffRegionScaleFactors=(PositiveWidth=1,NegativeWidth=2,PositiveHeight=2,NegativeHeight=1)
    FalloffSize=0.2
    FlipHeightScaling=false
    FlipWidthScaling=false
    OptionalParentItem=(Type=Bone,Name="None")
    RotationOffset=(X=65.2795,Y=-29.847,Z=13.0415)
   )
   > RigVMVariableNode(Value=@n18.OutputParam) {variable=thigh_side_in_l}
   > n19 = &RigUnit_SphericalPoseReader(
    ActiveRegionScaleFactors=(PositiveWidth=1,NegativeWidth=2,PositiveHeight=2,NegativeHeight=1)
    ActiveRegionSize=0.05
    Debug=(bDrawDebug=false,bDraw2D=False,bDrawLocalAxes=False,DebugScale=25,DebugSegments=20,DebugThickness=0.25)
    DriverAxis=(X=1,Y=0,Z=0)
    DriverItem=(Type=Bone,Name="thigh_r")
    FalloffRegionScaleFactors=(PositiveWidth=1,NegativeWidth=2,PositiveHeight=2,NegativeHeight=1)
    FalloffSize=0.2
    FlipHeightScaling=False
    FlipWidthScaling=true
    OptionalParentItem=(Type=Bone,Name="None")
    RotationOffset=(X=245.2795,Y=-29.847,Z=13.0415)
   )
   > RigVMVariableNode(Value=@n19.OutputParam) {variable=thigh_side_in_r}
   &RigUnit_ModifyTransforms(
    ItemToModify="((Item=(Type=Bone,Name=\"base_apron_fr_l\"),Transform=(Rotation=(X=0,Y=0,Z=0,W=-1),Translation=(X=0,Y=-2,Z=0),Scale3D=(X=1,Y=1,Z=1))),(Item=(Type=Bone,Name=\"base_apron_side_l\"),Transform=(Rotation=(X=0,Y=0,Z=0,W=1),Translation=(X=0,Y=-2,Z=0),Scale3D=(X=1,Y=1 ...(truncated)"
    WeightMinimum=0
    Weight=&RigVMFunction_MathBoolToFloat(Value=RigVMVariableNode() {variable=PandaSpeakEmote}.Value).Result
    ExecutePin=@RigVMFunction_Sequence_1.G
   )
   > ASW Apron Fix = RigVMFunctionReferenceNode()
   RigVMFunctionReturnNode(ExecuteContext=@RigVMFunction_Sequence_1.H)
   RigVMFunction_MathBoolNot = &RigVMFunction_MathBoolNot(Value=RigVMVariableNode() {variable=sliding_alpha}.Value)
   RigVMFunction_MathBoolAnd = RigVMAggregateNode(
    B=@RigVMFunction_MathBoolNot.Result
    C=&RigVMFunction_MathBoolNot(Value=RigVMVariableNode() {variable=baseball_sliding_alpha}.Value).Result
    D=&RigVMFunction_MathBoolNot(Value=RigVMVariableNode() {variable=moving2D_alpha}.Value).Result
    A=RigVMVariableNode() {variable=crouching_alpha}.Value
   )
   n20 = &RigUnit_GetRelativeTransformForItem(Child=(Type=Bone,Name="pelvis"), Parent=(Type=Bone,Name="root"), bChildInitial=False, bParentInitial=False)
   n21 = &RigVMFunction_AlphaInterp(
    InterpSpeedDecreasing=10
    InterpSpeedIncreasing=10
    Scale=1
    bClampResult=False
    Value=&RigVMFunction_MathDoubleMul(
     B=&RigVMFunction_MathBoolToFloat(Value=@RigVMFunction_MathBoolAnd.Result).Result
     A=&RigVMFunction_MathDoubleRemap(SourceMaximum=55, SourceMinimum=89, TargetMaximum=1, TargetMinimum=0, Value=@n20.RelativeTransform.Translation.Z).Result
    ).Result
   )
   RigVMVariableNode(Value=@n21.Result, ExecuteContext=@RigVMFunction_Sequence_1.D) {variable=crouching_not_walking_alpha}
   > &RigUnit_ModifyTransforms(
    ItemToModify=((Item=(Type=Bone,Name="base_apron_fr_c"),Transform=(Rotation=(X=0,Y=0,Z=0,W=1),Translation=(X=-7,Y=0,Z=0),Scale3D=(X=1,Y=1,Z=1))))
    WeightMinimum=0
    Weight=@n21.Result
   )
   > &RigUnit_ModifyTransforms(
    ItemToModify="((Item=(Type=Bone,Name=\"base_apron_fr_l\"),Transform=(Rotation=(X=0,Y=0,Z=0.1478,W=0.989),Translation=(X=-10,Y=-3,Z=0),Scale3D=(X=1,Y=1,Z=1))),(Item=(Type=Bone,Name=\"dyn_apron_side_l_00\"),Transform=(Rotation=(X=0.0594,Y=-0.1632,Z=0.3368,W=0.9254),Translatio ...(truncated)"
    WeightMinimum=0
    Weight=&RigVMFunction_MathDoubleMul(B=RigVMVariableNode() {variable=thigh_l_up_alpha}.Value, A=RigVMVariableNode() {variable=crouching_not_walking_alpha}.Value).Result
   )
   > &RigUnit_ModifyTransforms(
    ItemToModify="((Item=(Type=Bone,Name=\"base_apron_fr_l\"),Transform=(Rotation=(X=0.0049,Y=0.1391,Z=-0.0346,W=0.9897),Translation=(X=0,Y=0,Z=0),Scale3D=(X=1,Y=1,Z=1))),(Item=(Type=Bone,Name=\"dyn_apron_side_l_00\"),Transform=(Rotation=(X=0,Y=0,Z=-0.1305,W=0.9914),Translation ...(truncated)"
    WeightMinimum=0
    Weight=&RigVMFunction_MathDoubleMul(B=RigVMVariableNode() {variable=thigh_l_side_alpha}.Value, A=RigVMVariableNode() {variable=crouching_not_walking_alpha}.Value).Result
   )
   n22 = RigVMVariableNode() {variable=crouching_not_walking_alpha}
   > &RigUnit_ModifyTransforms(
    ItemToModify="((Item=(Type=Bone,Name=\"base_apron_fr_r\"),Transform=(Rotation=(X=0.0479,Y=-0.2544,Z=0.1685,W=0.9511),Translation=(X=10,Y=3,Z=0),Scale3D=(X=1,Y=1,Z=1))),(Item=(Type=Bone,Name=\"base_apron_side_r\"),Transform=(Rotation=(X=-0.0338,Y=0.0708,Z=0.4292,W=0.8998),Tr ...(truncated)"
    WeightMinimum=0
    Weight=&RigVMFunction_MathDoubleMul(B=RigVMVariableNode() {variable=thigh_r_up_alpha}.Value, A=@n22.Value).Result
   )
   > &RigUnit_ModifyTransforms(
    ItemToModify="((Item=(Type=Bone,Name=\"base_apron_fr_r\"),Transform=(Rotation=(X=0.0049,Y=0.1391,Z=-0.0346,W=0.9897),Translation=(X=0,Y=0,Z=0),Scale3D=(X=1,Y=1,Z=1))),(Item=(Type=Bone,Name=\"base_apron_side_r\"),Transform=(Rotation=(X=0,Y=0,Z=-0.165,W=0.9863),Translation=(X ...(truncated)"
    WeightMinimum=0
    Weight=@n22.Value
   )
   &RigUnit_ModifyTransforms(
    ItemToModify="((Item=(Type=Bone,Name=\"base_apron_fr_l\"),Transform=(Rotation=(X=-0.2734,Y=-0.1915,Z=-0.7722,W=0.5407),Translation=(X=-14,Y=0,Z=0),Scale3D=(X=1,Y=1,Z=1))),(Item=(Type=Bone,Name=\"base_apron_side_l\"),Transform=(Rotation=(X=0.0211,Y=-0.7623,Z=-0.0211,W=0.6465 ...(truncated)"
    WeightMinimum=0
    Weight=RigVMVariableNode() {variable=thigh_l_up_alpha}.Value
    ExecutePin=@RigVMFunction_Sequence_1.C
   )
   > &RigUnit_ModifyTransforms(
    ItemToModify="((Item=(Type=Bone,Name=\"base_apron_fr_r\"),Transform=(Rotation=(X=-0.2734,Y=-0.1915,Z=-0.7722,W=0.5407),Translation=(X=14,Y=0,Z=0),Scale3D=(X=1,Y=1,Z=1))),(Item=(Type=Bone,Name=\"base_apron_side_r\"),Transform=(Rotation=(X=0.0211,Y=-0.7623,Z=-0.0211,W=0.6465) ...(truncated)"
    WeightMinimum=0
    Weight=RigVMVariableNode() {variable=thigh_r_up_alpha}.Value
   )
   > &RigUnit_ModifyTransforms(
    ItemToModify=((Item=(Type=Bone,Name="base_apron_fr_c"),Transform=(Rotation=(X=0,Y=0,Z=-0.8039,W=0.5948),Translation=(X=-15,Y=0,Z=0),Scale3D=(X=1,Y=1,Z=1))))
    WeightMinimum=0
    Weight=&RigVMFunction_MathDoubleMax(B=RigVMVariableNode() {variable=thigh_r_up_alpha}.Value, A=RigVMVariableNode() {variable=thigh_l_up_alpha}.Value).Result
   )
   > &RigUnit_ModifyTransforms(
    ItemToModify="((Item=(Type=Bone,Name=\"base_apron_side_l\"),Transform=(Rotation=(X=-0.1712,Y=-0.0623,Z=-0.924,W=0.3363),Translation=(X=-10,Y=9,Z=0),Scale3D=(X=1,Y=1,Z=1))),(Item=(Type=Bone,Name=\"base_apron_fr_l\"),Transform=(Rotation=(X=-0.2817,Y=0.3444,Z=-0.3827,W=0.8097) ...(truncated)"
    WeightMinimum=0
    Weight=RigVMVariableNode() {variable=thigh_l_side_alpha}.Value
   )
   > &RigUnit_ModifyTransforms(
    ItemToModify="((Item=(Type=Bone,Name=\"base_apron_side_r\"),Transform=(Rotation=(X=-0.1712,Y=-0.0623,Z=-0.924,W=0.3363),Translation=(X=10,Y=9,Z=0),Scale3D=(X=1,Y=1,Z=1))),(Item=(Type=Bone,Name=\"base_apron_fr_r\"),Transform=(Rotation=(X=-0.2817,Y=0.3444,Z=-0.3827,W=0.8097), ...(truncated)"
    WeightMinimum=0
    Weight=RigVMVariableNode() {variable=thigh_r_side_alpha}.Value
   )
   RigVMFunction_MathBoolToFloat = &RigVMFunction_MathBoolToFloat(Value=&RigVMFunction_MathBoolAnd(A=RigVMVariableNode() {variable=crouching_alpha}.Value, B=RigVMVariableNode() {variable=moving2D_alpha}.Value).Result)
   n23 = &RigVMFunction_MathDoubleMul(B=@RigVMFunction_MathBoolToFloat.Result, A=RigVMVariableNode() {variable=thigh_side_in_l}.Value)
   > &RigUnit_ModifyTransforms(
    ItemToModify="((Item=(Type=Bone,Name=\"base_apron_side_r\"),Transform=(Rotation=(X=0,Y=0,Z=-0.0474,W=0.9989),Translation=(X=0.1113,Y=1.1703,Z=0),Scale3D=(X=1,Y=1,Z=1))),(Item=(Type=Bone,Name=\"base_apron_fr_r\"),Transform=(Rotation=(X=-0.2713,Y=-0.0097,Z=0.1271,W=-0.954),Tr ...(truncated)"
    WeightMinimum=0
    Weight=&RigVMFunction_MathDoubleMul(B=0.75, A=@n23.Result).Result
   )
   n24 = RigVMVariableNode() {variable=crouching_alpha}
   n25 = RigVMVariableNode() {variable=moving2D_alpha}
   n26 = &RigVMFunction_MathDoubleMul(
    B=&RigVMFunction_MathBoolToFloat(Value=&RigVMFunction_MathBoolAnd(A=@n24.Value, B=@n25.Value).Result).Result
    A=RigVMVariableNode() {variable=thigh_side_in_r}.Value
   )
   > &RigUnit_ModifyTransforms(
    ItemToModify="((Item=(Type=Bone,Name=\"base_apron_side_r\"),Transform=(Rotation=(X=0.0013,Y=0.0372,Z=0.0351,W=-0.9987),Translation=(X=0.0527,Y=0.7499,Z=0),Scale3D=(X=1,Y=1,Z=1))),(Item=(Type=Bone,Name=\"base_apron_fr_r\"),Transform=(Rotation=(X=0.0025,Y=0.1254,Z=0.0526,W=-0 ...(truncated)"
    WeightMinimum=0
    Weight=&RigVMFunction_MathDoubleMul(B=0.75, A=@n26.Result).Result
   )
   Maximum = &RigVMFunction_MathDoubleMax(B=RigVMVariableNode() {variable=thigh_side_in_l}.Value, A=RigVMVariableNode() {variable=thigh_l_down_mid_alpha}.Value)
   n27 = &RigVMFunction_MathBoolAnd(
    A=&RigVMFunction_MathBoolNot(Value=RigVMVariableNode() {variable=crouching_alpha}.Value).Result
    B=RigVMVariableNode() {variable=moving2D_alpha}.Value
   )
   > &RigUnit_ModifyTransforms(
    ItemToModify="((Item=(Type=Bone,Name=\"base_apron_side_r\"),Transform=(Rotation=(X=0,Y=0,Z=0,W=1),Translation=(X=0,Y=0,Z=0),Scale3D=(X=1,Y=1,Z=1))),(Item=(Type=Bone,Name=\"base_apron_fr_r\"),Transform=(Rotation=(X=0.0258,Y=-0.063,Z=-0.2978,W=0.9522),Translation=(X=0,Y=0,Z=0 ...(truncated)"
    WeightMinimum=0
    Weight=&RigVMFunction_MathDoubleMul(A=@Maximum.Result, B=&RigVMFunction_MathBoolToFloat(Value=@n27.Result).Result).Result
   )
   > &RigUnit_ModifyTransforms(
    ItemToModify="((Item=(Type=Bone,Name=\"base_apron_side_r\"),Transform=(Rotation=(X=0,Y=0,Z=0,W=1),Translation=(X=0,Y=0,Z=0),Scale3D=(X=1,Y=1,Z=1))),(Item=(Type=Bone,Name=\"base_apron_fr_r\"),Transform=(Rotation=(X=0,Y=0,Z=0,W=1),Translation=(X=0,Y=0,Z=0),Scale3D=(X=1,Y=1,Z= ...(truncated)"
    WeightMinimum=0
    Weight=RigVMVariableNode() {variable=thigh_side_in_r}.Value
   )
   n28 = RigVMVariableNode() {variable=sliding_alpha}
   n29 = &RigUnit_GetRelativeTransformForItem(Child=(Type=Bone,Name="pelvis"), Parent=(Type=Bone,Name="root"), bChildInitial=False, bParentInitial=False)
   n30 = &RigVMFunction_AlphaInterp(
    InterpSpeedDecreasing=10
    InterpSpeedIncreasing=10
    Scale=1
    bClampResult=False
    Value=&RigVMFunction_MathDoubleMul(
     B=&RigVMFunction_MathBoolToFloat(Value=@n28.Value).Result
     A=&RigVMFunction_MathDoubleRemap(SourceMaximum=55, SourceMinimum=89, TargetMaximum=1, TargetMinimum=0, Value=@n29.RelativeTransform.Translation.Z).Result
    ).Result
   )
   &RigUnit_ModifyTransforms(
    ItemToModify="((Item=(Type=Bone,Name=\"base_apron_fr_r\"),Transform=(Rotation=(X=0,Y=-0.1305,Z=0,W=0.9914),Translation=(X=0,Y=0,Z=0),Scale3D=(X=1,Y=1,Z=1))),(Item=(Type=Bone,Name=\"base_apron_side_r\"),Transform=(Rotation=(X=0,Y=0,Z=0.1736,W=0.9848),Translation=(X=0,Y=-8,Z= ...(truncated)"
    WeightMinimum=0
    Weight=@n30.Result
    ExecutePin=@RigVMFunction_Sequence_1.E
   )
   n31 = RigVMVariableNode() {variable=sliding_alpha}
   n32 = &RigUnit_GetRelativeTransformForItem(Child=(Type=Bone,Name="pelvis"), Parent=(Type=Bone,Name="root"), bChildInitial=False, bParentInitial=False)
   n33 = &RigVMFunction_AlphaInterp(
    InterpSpeedDecreasing=10
    InterpSpeedIncreasing=10
    Scale=1
    bClampResult=False
    Value=&RigVMFunction_MathDoubleMul(
     B=&RigVMFunction_MathBoolToFloat(Value=@n31.Value).Result
     A=&RigVMFunction_MathDoubleRemap(SourceMaximum=55, SourceMinimum=89, TargetMaximum=1, TargetMinimum=0, Value=@n32.RelativeTransform.Translation.Z).Result
    ).Result
   )
   > &RigUnit_ModifyTransforms(
    ItemToModify="((Item=(Type=Bone,Name=\"base_apron_fr_r\"),Transform=(Rotation=(X=-0.0166,Y=0.1728,Z=0.0944,W=0.9803),Translation=(X=0,Y=0,Z=0),Scale3D=(X=1,Y=1,Z=1))),(Item=(Type=Bone,Name=\"base_apron_side_r\"),Transform=(Rotation=(X=0,Y=0,Z=0,W=1),Translation=(X=0,Y=0,Z=0 ...(truncated)"
    WeightMinimum=0
    Weight=@n33.Result
   )
   > &RigUnit_ModifyTransforms(
    ItemToModify=((Item=(Type=Bone,Name="base_apron_fr_r"),Transform=(Rotation=(X=0.0046,Y=0.087,Z=-0.0521,W=0.9948),Translation=(X=0,Y=0,Z=0),Scale3D=(X=1,Y=1,Z=1))))
    WeightMinimum=0
    Weight=&RigVMFunction_MathDoubleMul(A=@n33.Result, B=RigVMVariableNode() {variable=thigh_r_up_alpha}.Value).Result
   )
   n34 = RigVMVariableNode() {variable=isKneeSliding}
   n35 = &RigUnit_GetRelativeTransformForItem(Child=(Type=Bone,Name="pelvis"), Parent=(Type=Bone,Name="root"), bChildInitial=False, bParentInitial=False)
   n36 = &RigVMFunction_AlphaInterp(
    InterpSpeedDecreasing=10
    InterpSpeedIncreasing=10
    Scale=1
    bClampResult=False
    Value=&RigVMFunction_MathDoubleMul(
     B=&RigVMFunction_MathBoolToFloat(Value=@n34.Value).Result
     A=&RigVMFunction_MathDoubleRemap(SourceMaximum=55, SourceMinimum=89, TargetMaximum=1, TargetMinimum=0, Value=@n35.RelativeTransform.Translation.Z).Result
    ).Result
   )
   > &RigUnit_ModifyTransforms(
    ItemToModify="((Item=(Type=Bone,Name=\"base_apron_fr_r\"),Transform=(Rotation=(X=0.0023,Y=0.0871,Z=-0.0261,W=0.9959),Translation=(X=0,Y=0,Z=0),Scale3D=(X=1,Y=1,Z=1))),(Item=(Type=Bone,Name=\"base_apron_side_r\"),Transform=(Rotation=(X=0,Y=0,Z=-0.2164,W=0.9763),Translation=( ...(truncated)"
    WeightMinimum=0
    Weight=@n36.Result
   )
   n37 = RigVMVariableNode() {variable=isPlayingEmote}
   > &RigUnit_ModifyTransforms(
    ItemToModify="((Item=(Type=Bone,Name=\"base_apron_fr_l\"),Transform=(Rotation=(X=-0.0019,Y=-0.0436,Z=-0.0436,W=0.9981),Translation=(X=0,Y=0,Z=0),Scale3D=(X=1,Y=1,Z=1))),(Item=(Type=Bone,Name=\"dyn_apron_side_l_00\"),Transform=(Rotation=(X=0,Y=0,Z=0,W=1),Translation=(X=0,Y=0 ...(truncated)"
    WeightMinimum=0
    Weight=&RigVMFunction_MathDoubleMul(A=&RigVMFunction_MathBoolToFloat(Value=@n37.Value).Result, B=RigVMVariableNode() {variable=thigh_l_up_alpha}.Value).Result
   )
   n38 = RigVMVariableNode() {variable=isPlayingEmote}
   > &RigUnit_ModifyTransforms(
    ItemToModify="((Item=(Type=Bone,Name=\"base_apron_fr_r\"),Transform=(Rotation=(X=-0.0019,Y=-0.0436,Z=-0.0436,W=0.9981),Translation=(X=0,Y=0,Z=0),Scale3D=(X=1,Y=1,Z=1))),(Item=(Type=Bone,Name=\"base_apron_side_r\"),Transform=(Rotation=(X=0,Y=0,Z=0,W=1),Translation=(X=0,Y=0,Z ...(truncated)"
    WeightMinimum=0
    Weight=&RigVMFunction_MathDoubleMul(A=&RigVMFunction_MathBoolToFloat(Value=@n38.Value).Result, B=RigVMVariableNode() {variable=thigh_r_up_alpha}.Value).Result
   )
   n39 = RigVMVariableNode() {variable=isCustomEmotePlaying}
   Maximum_1 = RigVMAggregateNode(
    A=RigVMVariableNode() {variable=thigh_l_down_mid_alpha}.Value
    B=RigVMVariableNode() {variable=thigh_side_in_l}.Value
    C=RigVMVariableNode() {variable=thigh_l_up_mid_alpha}.Value
   )
   > &RigUnit_ModifyTransforms(
    ItemToModify="((Item=(Type=Bone,Name=\"base_apron_fr_l\"),Transform=(Rotation=(X=0,Y=0,Z=0,W=1),Translation=(X=0,Y=0,Z=0),Scale3D=(X=1,Y=1,Z=1))),(Item=(Type=Bone,Name=\"dyn_apron_side_l_00\"),Transform=(Rotation=(X=0,Y=0,Z=0,W=1),Translation=(X=0,Y=0,Z=0),Scale3D=(X=1,Y=1, ...(truncated)"
    WeightMinimum=0
    Weight=&RigVMFunction_MathDoubleMul(A=&RigVMFunction_MathBoolToFloat(Value=@n39.Value).Result, B=@Maximum_1.Result).Result
   )
   n40 = RigVMVariableNode() {variable=isCustomEmotePlaying}
   Maximum_2 = RigVMAggregateNode(
    A=RigVMVariableNode() {variable=thigh_r_down_mid_alpha}.Value
    B=RigVMVariableNode() {variable=thigh_side_in_r}.Value
    C=RigVMVariableNode() {variable=thigh_r_up_mid_alpha}.Value
   )
   > &RigUnit_ModifyTransforms(
    ItemToModify="((Item=(Type=Bone,Name=\"base_apron_fr_l\"),Transform=(Rotation=(X=0,Y=0,Z=-0.2164,W=0.9763),Translation=(X=-5,Y=-5,Z=0),Scale3D=(X=1,Y=1,Z=1))),(Item=(Type=Bone,Name=\"base_apron_fr_c\"),Transform=(Rotation=(X=0,Y=0.1736,Z=0,W=0.9848),Translation=(X=-5,Y=-5,Z ...(truncated)"
    WeightMinimum=0
    Weight=&RigVMFunction_MathDoubleMul(A=&RigVMFunction_MathBoolToFloat(Value=@n40.Value).Result, B=@Maximum_2.Result).Result
   )
   n41 = RigVMVariableNode() {variable=isCustomEmotePlaying}
   > &RigUnit_ModifyTransforms(
    ItemToModify="((Item=(Type=Bone,Name=\"base_apron_fr_l\"),Transform=(Rotation=(X=0.0076,Y=0.0868,Z=-0.0868,W=0.9924),Translation=(X=0,Y=0,Z=0),Scale3D=(X=1,Y=1,Z=1))),(Item=(Type=Bone,Name=\"base_apron_side_l\"),Transform=(Rotation=(X=0,Y=0,Z=0,W=1),Translation=(X=0,Y=0,Z=0 ...(truncated)"
    WeightMinimum=0
    Weight=&RigVMFunction_MathDoubleMul(A=&RigVMFunction_MathBoolToFloat(Value=@n41.Value).Result, B=RigVMVariableNode() {variable=thigh_l_side_alpha}.Value).Result
   )
   n42 = RigVMVariableNode() {variable=isCustomEmotePlaying}
   > &RigUnit_ModifyTransforms(
    ItemToModify="((Item=(Type=Bone,Name=\"base_apron_fr_r\"),Transform=(Rotation=(X=0.056,Y=0.2527,Z=-0.2091,W=0.943),Translation=(X=0,Y=0,Z=0),Scale3D=(X=1,Y=1,Z=1))),(Item=(Type=Bone,Name=\"base_apron_side_r\"),Transform=(Rotation=(X=0,Y=0,Z=0,W=1),Translation=(X=0,Y=0,Z=0), ...(truncated)"
    WeightMinimum=0
    Weight=&RigVMFunction_MathDoubleMul(A=&RigVMFunction_MathBoolToFloat(Value=@n42.Value).Result, B=RigVMVariableNode() {variable=thigh_r_side_alpha}.Value).Result
   )
   RigVMFunction_MathBoolAnd_3 = RigVMAggregateNode(
    B=&RigVMFunction_MathBoolNot(Value=RigVMVariableNode() {variable=crouching_alpha}.Value).Result
    C=&RigVMFunction_MathBoolNot(Value=RigVMVariableNode() {variable=moving2D_alpha}.Value).Result
    A=RigVMVariableNode() {variable=isPlayingMelee}.Value
   )
   n43 = &RigVMFunction_MathBoolToFloat(Value=@RigVMFunction_MathBoolAnd_3.Result)
   > &RigUnit_ModifyTransforms(
    ItemToModify="((Item=(Type=Bone,Name=\"base_apron_fr_r\"),Transform=(Rotation=(X=0,Y=0,Z=-0.165,W=0.9863),Translation=(X=0,Y=0,Z=0),Scale3D=(X=1,Y=1,Z=1))),(Item=(Type=Bone,Name=\"base_apron_fr_l\"),Transform=(Rotation=(X=-0.003,Y=-0.0697,Z=-0.0435,W=0.9966),Translation=(X= ...(truncated)"
    WeightMinimum=0
    Weight=@n43.Result
   )
   RigVMFunction_MathBoolAnd_1 = RigVMAggregateNode(
    C=&RigVMFunction_MathBoolNot(Value=RigVMVariableNode() {variable=moving2D_alpha}.Value).Result
    A=RigVMVariableNode() {variable=isPlayingMelee}.Value
    B=RigVMVariableNode() {variable=crouching_alpha}.Value
   )
   Multiply = &RigVMFunction_MathDoubleMul(
    A=&RigVMFunction_MathBoolToFloat(Value=@RigVMFunction_MathBoolAnd_1.Result).Result
    B=RigVMVariableNode() {variable=thigh_l_up_alpha}.Value
   )
   > &RigUnit_ModifyTransforms(
    ItemToModify="((Item=(Type=Bone,Name=\"base_apron_fr_r\"),Transform=(Rotation=(X=0,Y=0,Z=0,W=1),Translation=(X=0,Y=0,Z=0),Scale3D=(X=1,Y=1,Z=1))),(Item=(Type=Bone,Name=\"base_apron_fr_l\"),Transform=(Rotation=(X=-0.0076,Y=-0.1735,Z=-0.043,W=0.9839),Translation=(X=0,Y=0,Z=0) ...(truncated)"
    WeightMinimum=0
    Weight=@Multiply.Result
   )
   &RigUnit_ParentConstraint(
    AdvancedSettings=(InterpolationType=Average,RotationOrderForFilter=XZY)
    Child=(Type=Bone,Name="base_apron_alt_helper")
    Parents=((Item=(Type=Bone,Name="dyn_apron_fr_c_01"),Weight=1))
    Weight=1
    ExecutePin=@RigVMFunction_Sequence_1.F
   )
   > &RigUnit_ModifyTransforms(
    ItemToModify=((Item=(Type=Bone,Name="base_apron_alt_helper"),Transform=(Rotation=(X=0,Y=0,Z=-0.0436,W=0.999),Translation=(X=0,Y=-2,Z=0),Scale3D=(X=1,Y=1,Z=1))))
    WeightMinimum=0
    Weight=RigVMVariableNode() {variable=thigh_r_up_alpha}.Value
   )
   > &RigUnit_ModifyTransforms(
    ItemToModify=((Item=(Type=Bone,Name="base_apron_alt_helper"),Transform=(Rotation=(X=0,Y=0,Z=0.0436,W=0.999),Translation=(X=0,Y=2,Z=0),Scale3D=(X=1,Y=1,Z=1))))
    WeightMinimum=0
    Weight=RigVMVariableNode() {variable=thigh_l_up_alpha}.Value
   )
   Clamp = &RigVMFunction_MathDoubleClamp(
    Maximum=0.2
    Value=&RigVMFunction_MathDoubleMul(B=RigVMVariableNode() {variable=Emote GoldenGlory}.Value, A=RigVMVariableNode() {variable=thigh_r_front_side_alpha}.Value).Result
   )
   > &RigUnit_ModifyTransforms(
    ItemToModify="((Item=(Type=Bone,Name=\"base_apron_fr_r\"),Transform=(Rotation=(X=0,Y=0,Z=0.0692,W=-0.9976),Translation=(X=0,Y=12.5694,Z=3.0151),Scale3D=(X=1,Y=1,Z=1))),(Item=(Type=Bone,Name=\"base_apron_side_r\"),Transform=(Rotation=(X=0,Y=0,Z=0,W=-1),Translation=(X=0,Y=5,Z ...(truncated)"
    WeightMinimum=0
    Weight=@Clamp.Result
   )
   > &RigUnit_ModifyTransforms(
    ItemToModify="((Item=(Type=Bone,Name=\"base_apron_fr_r\"),Transform=(Rotation=(X=0,Y=0,Z=0,W=1),Translation=(X=0,Y=0,Z=0),Scale3D=(X=1,Y=1,Z=1))),(Item=(Type=Bone,Name=\"base_apron_fr_l\"),Transform=(Rotation=(X=0,Y=0,Z=-0.41,W=0.9121),Translation=(X=0,Y=0,Z=0),Scale3D=(X=1 ...(truncated)"
    WeightMinimum=0
    Weight=&RigVMFunction_MathDoubleMul(B=RigVMVariableNode() {variable=Emote GoldenGlory}.Value, A=RigVMVariableNode() {variable=thigh_r_up_alpha}.Value).Result
   )
   Remap = &RigVMFunction_MathDoubleRemap(
    SourceMaximum=5
    SourceMinimum=0
    TargetMaximum=1
    TargetMinimum=0
    Value=&RigVMFunction_MathQuaternionToEuler(
     RotationOrder=ZYX
     Value=&RigUnit_GetRelativeTransformForItem(Child=(Type=Bone,Name="spine_01"), Parent=(Type=Bone,Name="pelvis"), bChildInitial=False, bParentInitial=False).RelativeTransform.Rotation
    ).Result.Y
   )
   > &RigUnit_ModifyTransforms(
    ItemToModify="((Item=(Type=Bone,Name=\"base_apron_side_l\"),Transform=(Rotation=(X=0,Y=0.0531,Z=0,W=-0.9986),Translation=(X=-0.5779,Y=-3.3124,Z=5.4199),Scale3D=(X=1,Y=1,Z=1))),(Item=(Type=Bone,Name=\"base_apron_fr_l\"),Transform=(Rotation=(X=0,Y=0,Z=0.0389,W=-0.9992),Transl ...(truncated)"
    WeightMinimum=0
    Weight=&RigVMFunction_MathDoubleMul(A=@Remap.Result, B=RigVMVariableNode() {variable=Emote GoldenGlory}.Value).Result
   )
   &RigUnit_ModifyTransforms(
    ExecutePin=()
    ItemToModify="((Item=(Type=Bone,Name=\"base_apron_fr_l\"),Transform=(Rotation=(X=-0.0001,Y=0.0815,Z=-0.0847,W=0.9931),Translation=(X=0,Y=0,Z=0),Scale3D=(X=1,Y=1,Z=1))),(Item=(Type=Bone,Name=\"base_apron_side_l\"),Transform=(Rotation=(X=-0.0037,Y=-0.0609,Z=-0.0609,W=0.9963), ...(truncated)"
    WeightMinimum=0
    Weight=RigVMVariableNode() {variable=thigh_l_front_side_alpha}.Value
   )
   > &RigUnit_ModifyTransforms(
    ItemToModify="((Item=(Type=Bone,Name=\"base_apron_fr_r\"),Transform=(Rotation=(X=-0.0001,Y=0.0815,Z=-0.0847,W=0.9931),Translation=(X=0,Y=0,Z=0),Scale3D=(X=1,Y=1,Z=1))),(Item=(Type=Bone,Name=\"base_apron_side_r\"),Transform=(Rotation=(X=-0.0037,Y=-0.0609,Z=-0.0609,W=0.9963), ...(truncated)"
    WeightMinimum=0
    Weight=RigVMVariableNode() {variable=thigh_r_front_side_alpha}.Value
   )
   n44 = &RigUnit_SphericalPoseReader(
    ActiveRegionScaleFactors=(PositiveWidth=1,NegativeWidth=1,PositiveHeight=1,NegativeHeight=1)
    ActiveRegionSize=0.05
    Debug=(bDrawDebug=false,bDraw2D=False,bDrawLocalAxes=False,DebugScale=25,DebugSegments=20,DebugThickness=0.25)
    DriverAxis=(X=0,Y=1,Z=0)
    DriverItem=(Type=Bone,Name="thigh_l")
    ExecutePin=()
    FalloffRegionScaleFactors=(PositiveWidth=1,NegativeWidth=1,PositiveHeight=0.9,NegativeHeight=0.9)
    FalloffSize=0.33
    FlipHeightScaling=False
    FlipWidthScaling=False
    OptionalParentItem=(Type=Bone,Name="None")
    RotationOffset=(X=35,Y=0,Z=0)
   )
   > RigVMVariableNode(Value=@n44.OutputParam) {variable=thigh_l_twist_in_alpha}
   > n45 = &RigUnit_SphericalPoseReader(
    ActiveRegionScaleFactors=(PositiveWidth=1,NegativeWidth=1,PositiveHeight=1,NegativeHeight=1)
    ActiveRegionSize=0.05
    Debug=(bDrawDebug=false,bDraw2D=False,bDrawLocalAxes=False,DebugScale=25,DebugSegments=20,DebugThickness=0.25)
    DriverAxis=(X=0,Y=-1,Z=0)
    DriverItem=(Type=Bone,Name="thigh_r")
    FalloffRegionScaleFactors=(PositiveWidth=1,NegativeWidth=1,PositiveHeight=0.9,NegativeHeight=0.9)
    FalloffSize=0.33
    FlipHeightScaling=False
    FlipWidthScaling=False
    OptionalParentItem=(Type=Bone,Name="None")
    RotationOffset=(X=215,Y=0,Z=0)
   )
   > RigVMVariableNode(Value=@n45.OutputParam) {variable=thigh_r_twist_in_alpha}
   &RigVMFunction_AlphaInterp(InterpSpeedDecreasing=30, InterpSpeedIncreasing=30, Scale=1, bClampResult=False, Value=@n23.Result)
   &RigVMFunction_AlphaInterp(InterpSpeedDecreasing=30, InterpSpeedIncreasing=30, Scale=1, bClampResult=False, Value=@n26.Result)
   &RigVMFunction_AlphaInterp(InterpSpeedDecreasing=10, InterpSpeedIncreasing=10, Scale=1, Value=0, bClampResult=False)
   &RigVMFunction_MathDoubleMul(A=@n43.Result, B=RigVMVariableNode() {variable=thigh_l_up_alpha}.Value)
   &RigVMFunction_MathDoubleMul(
    B=1
    A=&RigVMFunction_MathDoubleRemap(
     SourceMaximum=90.2
     SourceMinimum=94.4
     TargetMaximum=1
     TargetMinimum=0
     Value=&RigUnit_GetRelativeTransformForItem(Child=(Type=Bone,Name="pelvis"), Parent=(Type=Bone,Name="root"), bChildInitial=False, bParentInitial=False).RelativeTransform.Translation.Z
    ).Result
   )
   &RigVMFunction_MathDoubleMul(B=RigVMVariableNode() {variable=thigh_r_side_alpha}.Value, A=@n22.Value)
   RigVMFunction_MathQuaternionSwingTwist = &RigVMFunction_MathQuaternionSwingTwist(
    twistAxis=(X=1,Y=0,Z=0)
    Input=&RigUnit_GetRelativeTransformForItem(Child=(Type=Bone,Name="thigh_l"), Parent=(Type=Bone,Name="pelvis"), bChildInitial=False, bParentInitial=False).RelativeTransform.Rotation
   )
   RigVMFunction_MathQuaternionToEuler = &RigVMFunction_MathQuaternionToEuler(rotationOrder=ZYX, Value=@RigVMFunction_MathQuaternionSwingTwist.Swing)
   &RigVMFunction_MathDoubleRemap(SourceMaximum=3, SourceMinimum=0, TargetMaximum=1, TargetMinimum=0, Value=@RigVMFunction_MathQuaternionToEuler.Result.Y)
   RigVMVariableNode() {variable=isPlayingEmote}
   RigVMVariableNode() {variable=isPlayingEmote}
   RigVMVariableNode() {variable=isPlayingEmote}
   RigVMVariableNode() {variable=isPlayingEmote}
   RigVMVariableNode() {variable=crouching_alpha}
  ]}
 New Function_ContainedGraph {nodes [
   n1 = RigVMFunctionEntryNode()
   RigVMFunction_MathQuaternionToEuler = &RigVMFunction_MathQuaternionToEuler(
    RotationOrder=YXZ
    Value=&RigUnit_GetRelativeTransformForItem(
     Child=(Type=Bone,Name="None")
     Parent=(Type=Bone,Name="None")
     bChildInitial=False
     bParentInitial=False
     Child.Name=@n1.Child
     Parent.Name=@n1.Parent
    ).RelativeTransform.Rotation
   )
   Remap = &RigVMFunction_MathDoubleRemap(
    TargetMaximum=1
    TargetMinimum=0
    Value=@RigVMFunction_MathQuaternionToEuler.Result.Y
    SourceMaximum=@n1.SourceMax
    SourceMinimum=@n1.SourceMin
   )
   > RigVMFunctionReturnNode(DeltaY=@Remap.Result)
  ]}
 New Function_ContainedGraph {nodes [
   n1 = RigVMFunctionEntryNode()
   RigVMFunction_MathQuaternionToEuler = &RigVMFunction_MathQuaternionToEuler(
    RotationOrder=ZYX
    Value=&RigUnit_GetRelativeTransformForItem(
     Child=(Type=Bone,Name="None")
     Parent=(Type=Bone,Name="None")
     bChildInitial=False
     bParentInitial=False
     Child.Name=@n1.Child
     Parent.Name=@n1.Parent
    ).RelativeTransform.Rotation
   )
   Remap = &RigVMFunction_MathDoubleRemap(
    TargetMaximum=1
    TargetMinimum=0
    Value=@RigVMFunction_MathQuaternionToEuler.Result.Z
    SourceMaximum=@n1.SourceMax
    SourceMinimum=@n1.SourceMin
   )
   > RigVMFunctionReturnNode(Delta Z=@Remap.Result)
  ]}
 New Function_ContainedGraph {nodes [
   n1 = RigVMFunctionEntryNode()
   RigVMFunction_MathBoolNot = &RigVMFunction_MathBoolNot(Value=RigVMVariableNode() {variable=RBAN_Enable}.Value)
   RigVMFunction_ControlFlowBranch = &RigVMFunction_ControlFlowBranch(ExecuteContext=@n1.ExecuteContext, Condition=@RigVMFunction_MathBoolNot.Result)
   RigVMFunctionReturnNode(ExecuteContext=@RigVMFunction_ControlFlowBranch.Completed)
   n2 = &RigUnit_SphericalPoseReader(
    ActiveRegionScaleFactors=(PositiveWidth=1,NegativeWidth=1,PositiveHeight=1,NegativeHeight=1)
    ActiveRegionSize=0.15
    Debug=(bDrawDebug=true,bDraw2D=False,bDrawLocalAxes=False,DebugScale=25,DebugSegments=20,DebugThickness=0.25)
    DriverAxis=(X=1,Y=0,Z=0)
    DriverItem=(Type=Bone,Name="spine_02")
    FalloffRegionScaleFactors=(PositiveWidth=1,NegativeWidth=1,PositiveHeight=1,NegativeHeight=1)
    FalloffSize=0.4
    FlipHeightScaling=False
    FlipWidthScaling=False
    OptionalParentItem=(Type=Bone,Name="root")
    RotationOffset=(X=90,Y=0,Z=0)
    ExecutePin=@RigVMFunction_ControlFlowBranch.True
   )
   Multiply = &RigVMFunction_MathDoubleMul(A=@n2.OutputParam, B=RigVMVariableNode() {variable=Emote GoldenGlory}.Value)
   > &RigUnit_ModifyTransforms(
    ItemToModify=((Item=(Type=Bone,Name="base_apron_alt_helper"),Transform=(Rotation=(X=-0.0233,Y=0.1608,Z=-0.4659,W=-0.8698),Translation=(X=-7.2886,Y=15.0824,Z=2.2415),Scale3D=(X=1,Y=1,Z=1))))
    WeightMinimum=0
    Weight=@Multiply.Result
   )
   > n3 = &RigUnit_SphericalPoseReader(
    ActiveRegionScaleFactors=(PositiveWidth=1,NegativeWidth=1,PositiveHeight=1,NegativeHeight=1)
    ActiveRegionSize=0.05
    Debug=(bDrawDebug=false,bDraw2D=False,bDrawLocalAxes=False,DebugScale=25,DebugSegments=20,DebugThickness=0.25)
    DriverAxis=(X=1,Y=0,Z=0)
    DriverItem=(Type=Bone,Name="thigh_r")
    FalloffRegionScaleFactors=(PositiveWidth=1,NegativeWidth=1,PositiveHeight=1,NegativeHeight=1)
    FalloffSize=0.8
    FlipHeightScaling=false
    FlipWidthScaling=false
    OptionalParentItem=(Type=Bone,Name="None")
    RotationOffset=(X=-90,Y=0,Z=-30)
   )
   > &RigUnit_ModifyTransforms(
    ItemToModify=((Item=(Type=Bone,Name="base_apron_alt_helper"),Transform=(Rotation=(X=0.3541,Y=0.1172,Z=-0.3174,W=0.8718),Translation=(X=-11.2944,Y=-9.7194,Z=-9.6375),Scale3D=(X=1,Y=1,Z=1))))
    WeightMinimum=0
    Weight=&RigVMFunction_MathDoubleMul(A=@n3.OutputParam, B=RigVMVariableNode() {variable=Emote GoldenGlory}.Value).Result
   )
   > n4 = &RigUnit_SphericalPoseReader(
    ActiveRegionScaleFactors=(PositiveWidth=1,NegativeWidth=1,PositiveHeight=1,NegativeHeight=1)
    ActiveRegionSize=0.2
    Debug=(bDrawDebug=false,bDraw2D=False,bDrawLocalAxes=False,DebugScale=25,DebugSegments=20,DebugThickness=0.25)
    DriverAxis=(X=1,Y=0,Z=0)
    DriverItem=(Type=Bone,Name="upperarm_l")
    FalloffRegionScaleFactors=(PositiveWidth=1,NegativeWidth=1,PositiveHeight=1,NegativeHeight=1)
    FalloffSize=0.45
    FlipHeightScaling=False
    FlipWidthScaling=False
    OptionalParentItem=(Type=Bone,Name="spine_05")
    RotationOffset=(X=0,Y=0,Z=0)
   )
   > &RigUnit_ModifyTransforms(
    ItemToModify=((Item=(Type=Bone,Name="base_apron_alt_helper"),Transform=(Rotation=(X=-0.0428,Y=0.1035,Z=-0.153,W=0.9819),Translation=(X=0,Y=-7.5811,Z=0),Scale3D=(X=1,Y=1,Z=1))))
    WeightMinimum=0
    Weight=&RigVMFunction_MathDoubleMul(A=@n4.OutputParam, B=RigVMVariableNode() {variable=Emote GoldenGlory}.Value).Result
   )
   > DeltaRotation Z = RigVMFunctionReferenceNode(Child=thigh_r, Parent=pelvis, SourceMax=130, SourceMin=180)
   > DeltaRotation Z_1 = RigVMFunctionReferenceNode(Child=thigh_l, Parent=pelvis, SourceMax=-50, SourceMin=0)
   RigVMFunction_MathQuaternionToEuler = &RigVMFunction_MathQuaternionToEuler(
    RotationOrder=XYZ
    Value=&RigUnit_GetRelativeTransformForItem(Child=(Type=Bone,Name="pelvis"), Parent=(Type=Bone,Name="root"), bChildInitial=False, bParentInitial=False).RelativeTransform.Rotation
   )
   AlphaInterp = &RigVMFunction_AlphaInterp(
    InterpSpeedDecreasing=2
    InterpSpeedIncreasing=10
    Scale=1
    bClampResult=true
    Value=&RigVMFunction_MathDoubleRemap(SourceMaximum=90, SourceMinimum=60, TargetMaximum=1, TargetMinimum=0, Value=@RigVMFunction_MathQuaternionToEuler.Result.X).Result
   )
   Multiply_3 = RigVMAggregateNode(
    D=@AlphaInterp.Result
    A=@"DeltaRotation Z".Delta Z
    B=@"DeltaRotation Z_1".Delta Z
    C=RigVMVariableNode() {variable=Emote GoldenGlory}.Value
   )
   Remap = &RigVMFunction_MathDoubleRemap(SourceMaximum=0.35, SourceMinimum=0, TargetMaximum=1, TargetMinimum=0, Value=@Multiply_3.Result)
   > &RigUnit_ModifyTransforms(
    ItemToModify=((Item=(Type=Bone,Name="base_apron_alt_helper"),Transform=(Rotation=(X=-0.1449,Y=-0.0238,Z=0.334,W=0.9311),Translation=(X=-2.7146,Y=6.5388,Z=0.0003),Scale3D=(X=1.2142,Y=1,Z=1))))
    WeightMinimum=0
    Weight=@Remap.Result
   )
  ]}
 RigVMFunctionLibrary {nodes [
   ASW Apron Fix = RigVMCollapseNode(ExecuteContext=())
   DeltaRotation Y = RigVMCollapseNode(ExecuteContext=())
   DeltaRotation Z = RigVMCollapseNode(ExecuteContext=())
   apronSetup = RigVMCollapseNode(ExecuteContext=())
  ]}
 RigVMFunction_MathBoolAnd_1_ContainedGraph {nodes [
   n1 = RigVMFunctionEntryNode()
   RigVMFunction_MathBoolAnd = &RigVMFunction_MathBoolAnd(A=@n1.A, B=@n1.B)
   RigVMFunctionReturnNode(Result=&RigVMFunction_MathBoolAnd(B=@n1.D, A=&RigVMFunction_MathBoolAnd(B=@n1.C, A=@RigVMFunction_MathBoolAnd.Result).Result).Result)
  ]}
 RigVMFunction_MathBoolAnd_1_ContainedGraph {nodes [
   n1 = RigVMFunctionEntryNode()
   RigVMFunction_MathBoolAnd = &RigVMFunction_MathBoolAnd(A=@n1.A, B=@n1.B)
   RigVMFunctionReturnNode(Result=&RigVMFunction_MathBoolAnd(B=@n1.C, A=@RigVMFunction_MathBoolAnd.Result).Result)
  ]}
 RigVMFunction_MathBoolAnd_1_ContainedGraph {nodes [
   n1 = RigVMFunctionEntryNode()
   RigVMFunction_MathBoolAnd = &RigVMFunction_MathBoolAnd(A=@n1.A, B=@n1.B)
   RigVMFunctionReturnNode(Result=&RigVMFunction_MathBoolAnd(B=@n1.C, A=@RigVMFunction_MathBoolAnd.Result).Result)
  ]}
 RigVMFunction_Sequence_1_ContainedGraph {nodes [
   n1 = RigVMFunctionEntryNode()
   RigVMFunction_Sequence = &RigVMFunction_Sequence(ExecuteContext=@n1.ExecuteContext)
   n2 = &RigVMFunction_Sequence(ExecuteContext=@RigVMFunction_Sequence.B)
   n3 = &RigVMFunction_Sequence(ExecuteContext=@n2.B)
   n4 = &RigVMFunction_Sequence(ExecuteContext=@n3.B)
   n5 = &RigVMFunction_Sequence(ExecuteContext=@n4.B)
   n6 = &RigVMFunction_Sequence(ExecuteContext=@n5.B)
   n7 = &RigVMFunction_Sequence(ExecuteContext=@n6.B)
   RigVMFunctionReturnNode(A=@RigVMFunction_Sequence.A, B=@n2.A, C=@n3.A, D=@n4.A, E=@n5.A, F=@n6.A, G=@n7.A, H=@n7.B)
  ]}
 RigVMModel {nodes [
   RigUnit_BeginExecution = &RigUnit_BeginExecution()
   /Script/FortniteGame.RigUnit_FortDeformRig(
    Bool=(bUpperArm=true,bElbow=true,bWrist=true,bGroin=true,bGlute=true,bKnee=true,bPec=true,bBicep=False,bLat=False)
    Curve="(bLUpperArmUpAlpha=False,bRUpperArmUpAlpha=False,bLUpperArmFrAlpha=False,bRUpperArmFrAlpha=False,bLUpperArmBkAlpha=False,bRUpperArmBkAlpha=False,bLWristUpAlpha=False,bRWristUpAlpha=False,bLWristDnAlpha=False,bRWristDnAlpha=False,bLKneeAlpha=true,bRKneeAlph ...(truncated)"
    Debug=(bUpperArmDebug=False,bWristDebug=False,bLatDebug=False,bGroinGluteDebug=False)
    Value="(UpperArmUp=(Rotation=(X=0.0,Y=-0.0,Z=0.0,W=1.0),Translation=(X=0.0,Y=0.0,Z=8.0),Scale3D=(X=0.70,Y=1.0,Z=1.0)),UpperArmFwdFr=(Rotation=(X=0.0,Y=-0.0,Z=0.0,W=1.0),Translation=(X=0.0,Y=10.0,Z=0.0),Scale3D=(X=0.50,Y=1.0,Z=1.0)),UpperArmFwdBk=(Rotation=(X=0.0, ...(truncated)"
    ExecutePin=@RigUnit_BeginExecution.ExecutePin
   ) {is=struct RigVMUnitNode}
   > apronSetup = RigVMFunctionReferenceNode()
   > &RigUnit_ModifyTransforms(
    ItemToModify=((Item=(Type=Bone,Name="base_stomatchpush_01"),Transform=(Rotation=(X=0,Y=0,Z=0,W=1),Translation=(X=0,Y=-18.0803,Z=0),Scale3D=(X=1,Y=1,Z=1))))
    Weight=1
    WeightMinimum=0
   )
  ]}
]
hierarchy [
 root {loc=<0 0 0>, rot=<0 0 0 1>}
 pelvis {parent=root, loc=<0 -0.5301 98.0571>, rot=<0.0127 0.707 -0.0127 -0.707>}
 spine_01 {parent=pelvis, loc=<6.1208 -0.9328 0>, rot=<0 0 0.034 -0.9994>}
 spine_02 {parent=spine_01, loc=<6.5997 -0.2095 0>, rot=<0 0 0.0238 -0.9997>}
 spine_03 {parent=spine_02, loc=<6.4914 -0.1237 0>, rot=<0 0 0.0149 -0.9999>}
 spine_04 {parent=spine_03, loc=<6.9598 0.0794 0>, rot=<0 0 -0.0132 -0.9999>}
 spine_05 {parent=spine_04, loc=<6.9987 0 0>, rot=<0 0 -0.0015 -1>}
 neck_01 {parent=spine_05, loc=<11.3138 0.1529 0>, rot=<0 0 -0.2003 -0.9797>}
 neck_02 {parent=neck_01, loc=<4.5627 0 0>, rot=<0 0 0.0504 -0.9987>}
 head {parent=neck_02, loc=<5.3914 0 0>, rot=<0 0 0.115 -0.9934>}
 base_hair_00 {parent=head, loc=<5.9532 -7.6753 -0.0002>, rot=<0.0041 -1 0 0>}
 base_hair_01 {parent=base_hair_00, loc=<0 0 0>, rot=<0 0 0 1>}
 clavicle_l {parent=spine_05, loc=<5.4831 -0.1053 -3.2946>, rot=<0.0465 -0.7289 -0.009 -0.6829>}
 upperarm_l {parent=clavicle_l, loc=<9.8658 -1.2136 -0.4105>, rot=<0.0164 -0.3245 -0.0269 -0.9454>}
 lowerarm_l {parent=upperarm_l, loc=<26.465 -0.8345 0.259>, rot=<0.0036 0.0161 -0.2298 -0.9731>}
 hand_l {parent=lowerarm_l, loc=<22.7355 -0.3203 -0.6105>, rot=<-0.6784 -0.0223 0.0735 -0.7307>}
 thumb_01_l {parent=hand_l, loc=<1.3118 -0.3191 -2.459>, rot=<0.6546 -0.4042 0.0664 -0.6354>}
 thumb_02_l {parent=thumb_01_l, loc=<5.9336 -0.0448 0.0401>, rot=<-0.0037 0.0213 0.298 -0.9543>}
 thumb_03_l {parent=thumb_02_l, loc=<2.2223 0.3682 -0.0907>, rot=<0.0118 -0.013 0.0066 -0.9998>}
 index_metacarpal_l {parent=hand_l, loc=<1.6535 0.0183 -2.96>, rot=<0.0158 -0.0124 -0.0173 -0.9996>}
 index_01_l {parent=index_metacarpal_l, loc=<8.3008 0.0357 -0.0524>, rot=<-0.0044 0.0485 0.0924 -0.9945>}
 index_02_l {parent=index_01_l, loc=<3.3587 -0.3004 0.0452>, rot=<-0.002 0.0027 0.3225 -0.9466>}
 index_03_l {parent=index_02_l, loc=<2.5578 -0.123 -0.0252>, rot=<-0.002 0.0067 0.2558 -0.9667>}
 middle_metacarpal_l {parent=hand_l, loc=<2.1856 -0.1725 -0.5984>, rot=<0.006 0.0168 -0.0329 -0.9993>}
 middle_01_l {parent=middle_metacarpal_l, loc=<7.3275 -0.2447 0.0129>, rot=<-0.0208 0.0042 0.0858 -0.9961>}
 middle_02_l {parent=middle_01_l, loc=<4.0123 0.1401 0.1175>, rot=<-0.0218 0.0258 0.3512 -0.9357>}
 middle_03_l {parent=middle_02_l, loc=<2.6077 0.0534 0.1373>, rot=<0.0082 0.0188 0.1708 -0.9851>}
 ring_metacarpal_l {parent=hand_l, loc=<2.2074 -0.3625 1.3638>, rot=<-0.0284 0.0722 -0.0458 -0.9959>}
 ring_01_l {parent=ring_metacarpal_l, loc=<7.0295 -0.6197 -0.1446>, rot=<-0.0489 -0.0252 0.1892 -0.9804>}
 ring_02_l {parent=ring_01_l, loc=<3.6174 0.4647 -0.1027>, rot=<0 0 0.2806 -0.9598>}
 ring_03_l {parent=ring_02_l, loc=<2.4152 0.114 0.0199>, rot=<0 0 0.1189 -0.9929>}
 pinky_metacarpal_l {parent=hand_l, loc=<1.7602 -0.9753 3.3806>, rot=<-0.0397 0.0814 -0.069 -0.9935>}
 pinky_01_l {parent=pinky_metacarpal_l, loc=<7.1672 -0.4825 -0.1107>, rot=<-0.0654 -0.0035 0.2242 -0.9723>}
 pinky_02_l {parent=pinky_01_l, loc=<2.8256 0.0894 -0.1043>, rot=<0.0037 0.0011 0.2854 -0.9584>}
 pinky_03_l {parent=pinky_02_l, loc=<2.2033 0.0304 -0.0074>, rot=<-0.0002 0.0098 0.1445 -0.9895>, scale=<0.9999 1 1>}
 deform_wrist_dn_base_l {parent=hand_l, loc=<0 -1.5326 0>, rot=<0 0 0 -1>}
 deform_wrist_dn_l {parent=deform_wrist_dn_base_l, loc=<0 0 0>, rot=<0 0 0 1>}
 deform_wrist_up_base_l {parent=hand_l, loc=<0 2.3009 0>, rot=<0 0 0 -1>}
 deform_wrist_up_l {parent=deform_wrist_up_base_l, loc=<0 0 0>, rot=<0 0 0 1>}
 lowerarm_twist_02_l {parent=lowerarm_l, loc=<10.6707 0 0>, rot=<0 0 0 1>}
 lowerarm_twist_01_l {parent=lowerarm_l, loc=<18.316 0 0>, rot=<0 0 0 1>}
 deform_elbow_in_base_l {parent=lowerarm_l, loc=<0.1834 3.6062 -0.2572>, rot=<0 0 0 -1>}
 deform_elbow_in_l {parent=deform_elbow_in_base_l, loc=<0 0 0>, rot=<0 0 0 1>}
 deform_elbow_out_base_l {parent=lowerarm_l, loc=<0 -4 0>, rot=<0 0 0 1>}
 deform_elbow_out_l {parent=deform_elbow_out_base_l, loc=<0 0 0>, rot=<0 0 0 1>}
 upperarm_twist_01_l {parent=upperarm_l, loc=<7.2135 -0.1134 0>, rot=<0 0 0 1>}
 deform_upperarm_bk_base_l {parent=upperarm_twist_01_l, loc=<-6.4053 -4.3746 0.314>, rot=<0 0 0 -1>}
 deform_upperarm_bk_l {parent=deform_upperarm_bk_base_l, loc=<0 0 0>, rot=<0 0 0 1>}
 deform_upperarm_fr_base_l {parent=upperarm_twist_01_l, loc=<-5.6312 5.092 0.3354>, rot=<0 0 0 -1>}
 deform_upperarm_fr_l {parent=deform_upperarm_fr_base_l, loc=<0 0 0>, rot=<0 0 0 1>}
 deform_upperarm_up_base_l {parent=upperarm_twist_01_l, loc=<-8.891 0.4825 3.9517>, rot=<0 0 0 -1>}
 deform_upperarm_up_l {parent=deform_upperarm_up_base_l, loc=<0 0 0>, rot=<0 0 0 1>}
 upperarm_twist_02_l {parent=upperarm_l, loc=<19.7998 -0.1134 0>, rot=<0 0 0 1>}
 clavicle_r {parent=spine_05, loc=<5.4831 -0.1053 3.2946>, rot=<0.7289 0.0465 -0.6829 0.009>}
 upperarm_r {parent=clavicle_r, loc=<-9.8658 1.2136 0.4105>, rot=<0.0164 -0.3245 -0.0269 -0.9454>}
 lowerarm_r {parent=upperarm_r, loc=<-26.465 0.8345 -0.259>, rot=<0.0036 0.0161 -0.2298 -0.9731>}
 hand_r {parent=lowerarm_r, loc=<-22.7355 0.3203 0.6106>, rot=<-0.6784 -0.0223 0.0735 -0.7307>}
 thumb_01_r {parent=hand_r, loc=<-1.3118 0.3191 2.459>, rot=<0.6546 -0.4042 0.0664 -0.6354>}
 thumb_02_r {parent=thumb_01_r, loc=<-5.9336 0.0448 -0.0401>, rot=<-0.0037 0.0213 0.298 -0.9543>}
 thumb_03_r {parent=thumb_02_r, loc=<-2.2223 -0.3682 0.0907>, rot=<0.0118 -0.013 0.0066 -0.9998>}
 index_metacarpal_r {parent=hand_r, loc=<-1.6535 -0.0183 2.96>, rot=<0.0158 -0.0124 -0.0173 -0.9996>}
 index_01_r {parent=index_metacarpal_r, loc=<-8.3008 -0.0357 0.0524>, rot=<-0.0044 0.0485 0.0924 -0.9945>}
 index_02_r {parent=index_01_r, loc=<-3.3587 0.3004 -0.0452>, rot=<-0.002 0.0027 0.3225 -0.9466>}
 index_03_r {parent=index_02_r, loc=<-2.5578 0.123 0.0252>, rot=<-0.002 0.0067 0.2558 -0.9667>}
 middle_metacarpal_r {parent=hand_r, loc=<-2.1856 0.1725 0.5984>, rot=<0.006 0.0168 -0.0329 -0.9993>}
 middle_01_r {parent=middle_metacarpal_r, loc=<-7.3275 0.2447 -0.0129>, rot=<-0.0208 0.0042 0.0858 -0.9961>}
 middle_02_r {parent=middle_01_r, loc=<-4.0123 -0.1401 -0.1175>, rot=<-0.0218 0.0258 0.3512 -0.9357>}
 middle_03_r {parent=middle_02_r, loc=<-2.6077 -0.0534 -0.1373>, rot=<0.0082 0.0188 0.1708 -0.9851>}
 ring_metacarpal_r {parent=hand_r, loc=<-2.2074 0.3625 -1.3638>, rot=<-0.0284 0.0722 -0.0458 -0.9959>}
 ring_01_r {parent=ring_metacarpal_r, loc=<-7.0295 0.6197 0.1446>, rot=<-0.0489 -0.0252 0.1892 -0.9804>}
 ring_02_r {parent=ring_01_r, loc=<-3.6174 -0.4647 0.1027>, rot=<0 0 0.2806 -0.9598>}
 ring_03_r {parent=ring_02_r, loc=<-2.4152 -0.114 -0.0199>, rot=<0 0 0.1189 -0.9929>}
 pinky_metacarpal_r {parent=hand_r, loc=<-1.76 0.9746 -3.3802>, rot=<-0.0397 0.0814 -0.069 -0.9935>}
 pinky_01_r {parent=pinky_metacarpal_r, loc=<-7.1671 0.4832 0.1112>, rot=<-0.0654 -0.0035 0.2242 -0.9723>}
 pinky_02_r {parent=pinky_01_r, loc=<-2.8259 -0.0891 0.1049>, rot=<0.0037 0.0011 0.2854 -0.9584>}
 pinky_03_r {parent=pinky_02_r, loc=<-2.2032 -0.0311 0.0078>, rot=<-0.0002 0.0098 0.1445 -0.9895>, scale=<1.0001 0.9999 1>}
 deform_wrist_dn_base_r {parent=hand_r, loc=<-0.0001 1.5328 0>, rot=<0 0 0 -1>}
 deform_wrist_dn_r {parent=deform_wrist_dn_base_r, loc=<0 0 0>, rot=<0 0 0 1>}
 deform_wrist_up_base_r {parent=hand_r, loc=<-0.0002 -2.3007 -0.0001>, rot=<0 0 0 -1>}
 deform_wrist_up_r {parent=deform_wrist_up_base_r, loc=<0 0 0>, rot=<0 0 0 1>}
 lowerarm_twist_02_r {parent=lowerarm_r, loc=<-10.6707 0 0>, rot=<0 0 0 1>}
 lowerarm_twist_01_r {parent=lowerarm_r, loc=<-18.316 0 0>, rot=<0 0 0 1>}
 deform_elbow_in_base_r {parent=lowerarm_r, loc=<-0.1834 -3.6062 0.2572>, rot=<0 0 0 -1>}
 deform_elbow_in_r {parent=deform_elbow_in_base_r, loc=<0 0 0>, rot=<0 0 0 1>}
 deform_elbow_out_base_r {parent=lowerarm_r, loc=<0 4 -0.0001>, rot=<0 0 0 -1>}
 deform_elbow_out_r {parent=deform_elbow_out_base_r, loc=<0 0 0>, rot=<0 0 0 1>}
 upperarm_twist_01_r {parent=upperarm_r, loc=<-7.2135 0.1134 0>, rot=<0 0 0 1>}
 deform_upperarm_bk_base_r {parent=upperarm_twist_01_r, loc=<6.4053 4.3746 -0.314>, rot=<0 0 0 -1>}
 deform_upperarm_bk_r {parent=deform_upperarm_bk_base_r, loc=<0 0 0>, rot=<0 0 0 1>}
 deform_upperarm_fr_base_r {parent=upperarm_twist_01_r, loc=<5.6312 -5.092 -0.3354>, rot=<0 0 0 -1>}
 deform_upperarm_fr_r {parent=deform_upperarm_fr_base_r, loc=<0 0 0>, rot=<0 0 0 1>}
 deform_upperarm_up_base_r {parent=upperarm_twist_01_r, loc=<8.891 -0.4825 -3.9517>, rot=<0 0 0 -1>}
 deform_upperarm_up_r {parent=deform_upperarm_up_base_r, loc=<0 0 0>, rot=<0 0 0 1>}
 upperarm_twist_02_r {parent=upperarm_r, loc=<-19.7998 0.1134 0>, rot=<0 0 0 1>}
 attach_backpack {parent=spine_05, loc=<-0.1225 -13.4695 -0.0001>, rot=<-0.0093 -0.707 0.0093 -0.707>, scale=<0.9999 1.0001 1>}
 Backpack_BR {parent=attach_backpack, type=Null, loc=<0 3 0>, rot=<-0.0698 0 0 0.9976>}
 attach_cape {parent=spine_05, loc=<9.0162 -11.711 0.0012>, rot=<0.0616 -0.7044 -0.0616 -0.7044>, scale=<0.9999 1.0001 1>}
 cape {parent=attach_cape, type=Null, loc=<0 1 -3>, rot=<0.0872 0 0 0.9962>}
 deform_pec_aim_base_l {parent=spine_04, loc=<3.2456 9.3309 -1.2237>, rot=<0.0603 -0.4033 0.3318 -0.8507>}
 deform_pec_aim_l {parent=deform_pec_aim_base_l, loc=<0 0 0>, rot=<0 0 0 1>}
 deform_pec_base_l {parent=deform_pec_aim_l, loc=<9.3665 0 0>, rot=<0 0 0 1>}
 deform_pec_l {parent=deform_pec_base_l, loc=<0 0 0>, rot=<0 0 0 1>}
 deform_pec_aim_base_r {parent=spine_04, loc=<3.2452 9.3309 1.2237>, rot=<-0.4033 -0.0603 0.8507 0.3318>}
 deform_pec_aim_r {parent=deform_pec_aim_base_r, loc=<0 0 0>, rot=<0 0 0 1>}
 deform_pec_base_r {parent=deform_pec_aim_r, loc=<-9.3667 -0.0003 -0.0004>, rot=<0 0 0 1>}
 deform_pec_r {parent=deform_pec_base_r, loc=<0 0 0>, rot=<0 0 0 1>}
 thigh_l {parent=pelvis, loc=<-1.0855 -0.2109 -8.7053>, rot=<0.0009 0.0355 0.0155 -0.9993>}
 calf_l {parent=thigh_l, loc=<-42.5284 -0.0019 -0.7297>, rot=<0.0006 -0.0161 -0.0391 -0.9991>}
 foot_l {parent=calf_l, loc=<-46.7401 -2.5014 -0.592>, rot=<0.0427 -0.0238 0.0353 -0.9982>}
 ball_l {parent=foot_l, loc=<-3.5658 10.8066 -0.1544>, rot=<0.0001 -0.0001 -0.7063 -0.7079>}
 calf_twist_02_l {parent=calf_l, loc=<-19.1546 -1.7943 -0.2638>, rot=<0.0237 0.0095 -0.0035 -0.9997>}
 calf_twist_01_l {parent=calf_l, loc=<-35.8483 -1.9197 -0.5794>, rot=<0.0237 0.0095 -0.0035 -0.9997>}
 deform_knee_in_base_l {parent=calf_l, loc=<0.8031 -5.0231 0.2456>, rot=<0 0 0 -1>}
 deform_knee_in_l {parent=deform_knee_in_base_l, loc=<0 0 0>, rot=<0 0 0 1>}
 deform_knee_out_base_l {parent=calf_l, loc=<0 4 0>, rot=<0 0 0 -1>}
 deform_knee_out_l {parent=deform_knee_out_base_l, loc=<0 0 0>, rot=<0 0 0 1>}
 thigh_twist_01_l {parent=thigh_l, loc=<-27.952 0 0>, rot=<0.0013 0 0 -1>}
 deform_groin_base_l {parent=thigh_l, loc=<0.4801 5.6776 -0.2486>, rot=<0 0 0 -1>}
 deform_groin_l {parent=deform_groin_base_l, loc=<0 0 0>, rot=<0 0 0 1>}
 thigh_r {parent=pelvis, loc=<-1.0855 -0.2109 8.7053>, rot=<0.0355 -0.0009 0.9993 0.0155>}
 calf_r {parent=thigh_r, loc=<42.5284 0.0019 0.7297>, rot=<0.0006 -0.0161 -0.0391 -0.9991>}
 foot_r {parent=calf_r, loc=<46.7401 2.5014 0.592>, rot=<0.0427 -0.0238 0.0353 -0.9982>}
 ball_r {parent=foot_r, loc=<3.5658 -10.8066 0.1544>, rot=<0.0001 -0.0001 -0.7063 -0.7079>}
 calf_twist_02_r {parent=calf_r, loc=<19.1546 1.7943 0.2638>, rot=<0.0237 0.0095 -0.0035 -0.9997>}
 calf_twist_01_r {parent=calf_r, loc=<35.8483 1.9197 0.5794>, rot=<0.0237 0.0095 -0.0035 -0.9997>}
 deform_knee_in_base_r {parent=calf_r, loc=<-0.8031 5.0231 -0.2457>, rot=<0 0 0 -1>}
 deform_knee_in_r {parent=deform_knee_in_base_r, loc=<0 0 0>, rot=<0 0 0 1>}
 deform_knee_out_base_r {parent=calf_r, loc=<0 -4 0>, rot=<0 0 0 -1>}
 deform_knee_out_r {parent=deform_knee_out_base_r, loc=<0 0 0>, rot=<0 0 0 1>}
 thigh_twist_01_r {parent=thigh_r, loc=<27.952 0 0>, rot=<0.0013 0 0 -1>}
 deform_groin_base_r {parent=thigh_r, loc=<-0.4801 -5.6776 0.2486>, rot=<0 0 0 -1>}
 deform_groin_r {parent=deform_groin_base_r, loc=<0 0 0>, rot=<0 0 0 1>}
 deform_glute_base_l {parent=pelvis, loc=<-7.3466 -9.0213 -9.1245>, rot=<0.0172 0.031 -0.4477 -0.8935>}
 deform_glute_l {parent=deform_glute_base_l, loc=<0 0 0>, rot=<0 0 0 1>}
 deform_glute_base_r {parent=pelvis, loc=<-7.3464 -9.0213 9.1245>, rot=<0.031 -0.0172 0.8935 -0.4477>}
 deform_glute_r {parent=deform_glute_base_r, loc=<0 0 0>, rot=<0 0 0 1>}
 base_apron_fr_c {parent=pelvis, loc=<-6.3866 11.4036 0.0011>, rot=<0 0 0.9991 0.0434>}
 dyn_apron_fr_c_00 {parent=base_apron_fr_c, loc=<0 0 0>, rot=<0 0 0 1>}
 dyn_apron_fr_c_01 {parent=dyn_apron_fr_c_00, loc=<10.7998 0.0002 0.0001>, rot=<0 0 0 1>}
 dyn_apron_fr_c_02 {parent=dyn_apron_fr_c_01, loc=<10.7998 0.0002 0.0001>, rot=<0 0 0 1>}
 dyn_apron_fr_c_03 {parent=dyn_apron_fr_c_02, loc=<10.7998 0.0002 0.0001>, rot=<0 0 0 1>}
 dyn_apron_fr_c_04 {parent=dyn_apron_fr_c_03, loc=<10.7998 0.0002 0.0001>, rot=<0 0 0 1>}
 base_apron_fr_l {parent=pelvis, loc=<-5.209 8.8292 -10.4314>, rot=<-0.0059 0.2926 0.9556 0.0329>}
 dyn_apron_fr_l_00 {parent=base_apron_fr_l, loc=<0 0 0>, rot=<0 0 0 1>}
 dyn_apron_fr_l_01 {parent=dyn_apron_fr_l_00, loc=<11.1998 0.0001 0.0002>, rot=<0 0 0 1>}
 dyn_apron_fr_l_02 {parent=dyn_apron_fr_l_01, loc=<11.1998 0.0001 0.0002>, rot=<0 0 0 1>}
 dyn_apron_fr_l_03 {parent=dyn_apron_fr_l_02, loc=<11.1998 0.0001 0.0002>, rot=<0 0 0 1>}
 dyn_apron_fr_l_04 {parent=dyn_apron_fr_l_03, loc=<11.1998 0.0001 0.0002>, rot=<0 0 0 1>}
 base_apron_side_l {parent=pelvis, loc=<-1.9599 -3.6827 -16.2385>, rot=<0.0172 -0.7071 -0.7058 -0.0391>}
 dyn_apron_side_l_00 {parent=base_apron_side_l, loc=<0 0 0>, rot=<0 0 0 1>}
 dyn_apron_side_l_01 {parent=dyn_apron_side_l_00, loc=<11.9998 0.0002 0.0001>, rot=<0 0 0 1>}
 dyn_apron_side_l_02 {parent=dyn_apron_side_l_01, loc=<11.9998 0.0002 0.0001>, rot=<0 0 0 1>}
 dyn_apron_side_l_03 {parent=dyn_apron_side_l_02, loc=<11.9998 0.0002 0.0001>, rot=<0 0 0 1>}
 dyn_apron_side_l_04 {parent=dyn_apron_side_l_03, loc=<11.9998 0.0002 0.0001>, rot=<0 0 0 1>}
 base_apron_fr_r {parent=pelvis, loc=<-5.209 8.8292 10.4314>, rot=<-0.2926 -0.0059 0.0329 -0.9556>}
 dyn_apron_fr_r_00 {parent=base_apron_fr_r, loc=<0 0 0>, rot=<0 0 0 1>}
 dyn_apron_fr_r_01 {parent=dyn_apron_fr_r_00, loc=<-11.1998 -0.0001 -0.0002>, rot=<0 0 0 1>}
 dyn_apron_fr_r_02 {parent=dyn_apron_fr_r_01, loc=<-11.1997 0 -0.0002>, rot=<0 0 0 1>}
 dyn_apron_fr_r_03 {parent=dyn_apron_fr_r_02, loc=<-11.1998 -0.0001 -0.0002>, rot=<0 0 0 1>}
 dyn_apron_fr_r_04 {parent=dyn_apron_fr_r_03, loc=<-11.1997 -0.0001 -0.0002>, rot=<0 0 0 1>}
 base_apron_side_r {parent=pelvis, loc=<-1.9599 -3.6826 16.2385>, rot=<-0.7071 -0.0172 0.0391 -0.7058>}
 dyn_apron_side_r_00 {parent=base_apron_side_r, loc=<0 0 0>, rot=<0 0 0 1>}
 dyn_apron_side_r_01 {parent=dyn_apron_side_r_00, loc=<-11.9998 -0.0002 -0.0001>, rot=<0 0 0 1>}
 dyn_apron_side_r_02 {parent=dyn_apron_side_r_01, loc=<-11.9998 -0.0002 -0.0001>, rot=<0 0 0 1>}
 dyn_apron_side_r_03 {parent=dyn_apron_side_r_02, loc=<-11.9998 -0.0002 -0.0001>, rot=<0 0 0 1>}
 dyn_apron_side_r_04 {parent=dyn_apron_side_r_03, loc=<-11.9997 -0.0002 -0.0001>, rot=<0 0 0 1>}
 base_tie {parent=pelvis, loc=<4.5015 -16.4934 0>, rot=<-0.0736 0.7033 0.7033 -0.0736>}
 base_tie_l_loop {parent=base_tie, loc=<0.0447 -1.3257 -1.2004>, rot=<-0.5228 -0.0019 0.284 -0.8038>}
 dyn_tie_l_loop_01 {parent=base_tie_l_loop, loc=<0 0 0>, rot=<0 0 0 1>}
 dyn_tie_l_loop_02 {parent=dyn_tie_l_loop_01, loc=<3.1986 0 0>, rot=<0.107 -0.1537 -0.1046 -0.9767>}
 base_tie_l_string {parent=base_tie, loc=<3.1019 -0.4309 -1.0095>, rot=<0.0422 0.0132 0.0441 -0.998>}
 dyn_tie_l_string_01 {parent=base_tie_l_string, loc=<0 0 0>, rot=<0 0 0 1>}
 dyn_tie_l_string_02 {parent=dyn_tie_l_string_01, loc=<3.5228 0 0>, rot=<0.0046 0.0464 -0.0222 -0.9987>}
 dyn_tie_l_string_03 {parent=dyn_tie_l_string_02, loc=<5.2343 -0.0002 0>, rot=<0.0026 0.052 -0.0288 -0.9982>}
 base_tie_r_loop {parent=base_tie, loc=<2.7852 1.6301 -0.9617>, rot=<0.2961 -0.0134 -0.203 -0.9332>}
 dyn_tie_r_loop_01 {parent=base_tie_r_loop, loc=<0 0 0>, rot=<0 0 0 1>}
 dyn_tie_r_loop_02 {parent=dyn_tie_r_loop_01, loc=<3.2684 -0.0001 0>, rot=<-0.01 -0.0569 -0.0052 -0.9983>}
 base_tie_r_string {parent=base_tie, loc=<3.4053 1.0595 -1.2792>, rot=<0.0767 -0.0046 -0.0277 -0.9967>}
 dyn_tie_r_string_01 {parent=base_tie_r_string, loc=<0 0 0>, rot=<0 0 0 1>}
 dyn_tie_r_string_02 {parent=dyn_tie_r_string_01, loc=<3.425 -0.0001 0>, rot=<0.0006 0.04 0.0028 -0.9992>}
 base_pen {parent=pelvis, loc=<-10.5674 11.3861 8.4799>, rot=<-0.0578 -0.982 -0.1724 0.0513>}
 dyn_pen {parent=base_pen, loc=<0 0 0>, rot=<0 0 0 1>}
 thigh_l_null {parent=pelvis, type=Null, loc=<-1.0855 -0.2109 -8.7053>, rot=<0.0009 0.0355 0.0155 -0.9993>}
 thigh_r_null {parent=pelvis, type=Null, loc=<-1.0855 -0.2109 8.7053>, rot=<0.0355 -0.0009 0.9993 0.0155>}
 base_stomatchpush_01 {parent=pelvis, loc=<0.2645 7.3776 0>, rot=<0 0 0 1>}
 base_apron_alt_helper {parent=pelvis, loc=<-21.8006 13.6319 0>, rot=<0.0305 0.7065 0.7065 0.0305>}
 apron_alt_helper {parent=base_apron_alt_helper, loc=<15.7765 0 0>, rot=<0 0 0 1>}
]
variables [
 Emote GoldenGlory {type=double, default=0}
 PandaSpeakEmote {type=bool, default=False}
 RBAN_Enable {type=bool, default=False}
 baseball_sliding_alpha {type=bool, default=False}
 crouching_alpha {type=bool, default=False}
 crouching_not_walking_alpha {type=float, default=0}
 isCustomEmotePlaying {type=bool, default=False}
 isFalling {type=bool, default=False}
 isFrontEnd {type=bool, default=False}
 isKneeSliding {type=bool, default=False}
 isParachuteOpen {type=bool, default=False}
 isPlayingEmote {type=bool, default=False}
 isPlayingMelee {type=bool, default=False}
 isWearingCape {type=bool, default=False}
 moving2D_alpha {type=bool, default=False}
 sliding_alpha {type=bool, default=False}
 thigh_l_down_mid_alpha {type=float, default=0}
 thigh_l_front_side_alpha {type=float, default=0}
 thigh_l_side_alpha {type=float, default=0}
 thigh_l_twist_in_alpha {type=float, default=0}
 thigh_l_up_alpha {type=float, default=0}
 thigh_l_up_mid_alpha {type=float, default=0}
 thigh_r_down_mid_alpha {type=float, default=0}
 thigh_r_front_side_alpha {type=float, default=0}
 thigh_r_side_alpha {type=float, default=0}
 thigh_r_twist_in_alpha {type=float, default=0}
 thigh_r_up_alpha {type=float, default=0}
 thigh_r_up_mid_alpha {type=float, default=0}
 thigh_side_in_l {type=float, default=0}
 thigh_side_in_r {type=float, default=0}
]
