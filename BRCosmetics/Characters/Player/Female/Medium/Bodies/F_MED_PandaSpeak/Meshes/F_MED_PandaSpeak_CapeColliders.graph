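# Physics asset of capsule colliders for the F_MED_PandaSpeak female player body, named as collision for its cape.
# Capsules cover the head, clavicles, spine, pelvis and feet, with paired capsules on the middle spine bones.
# More capsules follow the tie's base, loop and string bones.
# Arm, thigh, calf and neck bodies hold no shapes, and no constraints join any bodies, so the asset defines collision volumes only.
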PhysicsAsset /BRCosmetics/Characters/Player/Female/Medium/Bodies/F_MED_PandaSpeak/Meshes/F_MED_PandaSpeak_CapeColliders
bodies {
 base_tie [{
   shapeType=Capsule
   length=12.6348
   radius=7.1006
   shapeName=base_tie_capsule
   center=<3.1376 -0.0476 4.5738>
   rotation=<rot 9.2858 0.6923 -89.8883>
  }]
 base_tie_l_loop [{shapeType=Capsule, length=0.5, radius=0.5, shapeName=base_tie_l_loop_capsule, rotation=<rot 0 -90 0>}]
 base_tie_l_string [{shapeType=Capsule, length=0.5, radius=0.5, shapeName=base_tie_l_string_capsule, rotation=<rot 0 -90 0>}]
 base_tie_r_loop [{shapeType=Capsule, length=0.5, radius=0.5, shapeName=base_tie_r_loop_capsule, rotation=<rot 0 -90 0>}]
 base_tie_r_string [{shapeType=Capsule, length=0.5, radius=0.5, shapeName=base_tie_r_string_capsule, rotation=<rot 0 -90 0>}]
 calf_l []
 calf_r []
 clavicle_l [{shapeType=Capsule, length=6.5623, radius=7.5194, center=<3.9566 -1.0862 -2.1127>, rotation=<rot -77.5566 177.3915 7.7314>}]
 clavicle_r [{shapeType=Capsule, length=6.5623, radius=7.5194, center=<-3.9566 1.0862 2.1127>, rotation=<rot -77.5566 177.3915 -172.2686>}]
 dyn_tie_l_loop_01 [{
   shapeType=Capsule
   length=2.1394
   radius=2.3506
   shapeName=dyn_tie_l_loop_01_capsule
   center=<2.1065 0.2893 -0.2286>
   rotation=<rot 8.2328 -103.7319 -88.1228>
  }]
 dyn_tie_l_loop_02 [{shapeType=Capsule, length=6.1119, radius=3.0287, shapeName=dyn_tie_l_loop_02_capsule, center=<-0.062 0.6918 0>, rotation=<rot 90 -11.37 0>}]
 dyn_tie_l_string_01 [{shapeType=Capsule, length=4.5658, radius=1.4362, shapeName=dyn_tie_l_string_02_capsule, center=<1.4779 0 -0.5807>, rotation=<rot 90 0 0>}]
 dyn_tie_l_string_02 [{shapeType=Capsule, length=7.7045, radius=1.4362, shapeName=dyn_tie_l_string_02_capsule, center=<1.4779 0 -0.5807>, rotation=<rot 90 0 0>}]
 dyn_tie_l_string_03 [{shapeType=Capsule, length=7.7045, radius=1.4362, shapeName=dyn_tie_l_string_03_capsule, center=<1.4779 0 -0.5807>, rotation=<rot 90 0 0>}]
 dyn_tie_r_loop_01 [{
   shapeType=Capsule
   length=2.9148
   radius=2.3465
   shapeName=dyn_tie_r_loop_01_capsule
   center=<1.705 -0.0286 -0.1119>
   rotation=<rot -3.8809 -90.5406 -87.2246>
  }]
 dyn_tie_r_loop_02 [{shapeType=Capsule, length=6.5267, radius=3, shapeName=dyn_tie_r_loop_02_capsule, center=<0.7239 -0.1561 -0.4345>, rotation=<rot 90 0 0>}]
 dyn_tie_r_string_01 [{shapeType=Capsule, length=4.5863, radius=1.43, shapeName=dyn_tie_r_string_02_capsule, center=<1 0 -0.2808>, rotation=<rot 90 0 0>}]
 dyn_tie_r_string_02 [{shapeType=Capsule, length=7.8734, radius=1.43, shapeName=dyn_tie_r_string_02_capsule, center=<1 0 -0.2808>, rotation=<rot 90 0 0>}]
 foot_l [{shapeType=Capsule, length=11.0765, radius=8.0344, center=<-4.1811 6.0027 -0.338>, rotation=<rot 0.0388 1.4701 -89.3982>}]
 foot_r [{shapeType=Capsule, length=11.0765, radius=8.0344, center=<4.1811 -6.0027 0.338>, rotation=<rot 0.0388 1.4701 90.6018>}]
 head [{shapeType=Capsule, length=0, radius=18, center=<9.0697 0.0746 0>, rotation=<rot 0 -90 0>}]
 lowerarm_l []
 lowerarm_r []
 neck_02 []
 pelvis [{shapeType=Capsule, length=9.6042, radius=14.5259, center=<-6.6675 -3.9135 -0.1981>}]
 spine_01 [{shapeType=Capsule, length=4.3888, radius=13.8843, center=<-8.9021 -3.0316 0>}]
 spine_02 [
  {shapeType=Capsule, length=3.2216, radius=11.3778, center=<-2.6885 -2.1586 0>}
  {shapeType=Capsule, length=21.7618, radius=9.3125, center=<8.7382 -3.5553 0>, rotation=<rot 90 11.5674 0>}
 ]
 spine_03 [
  {shapeType=Capsule, length=6.1863, radius=11.5898, center=<0.0574 -0.2541 0>}
  {shapeType=Capsule, length=20.7188, radius=9.2812, center=<-1.2163 0.549 0>, rotation=<rot 90 180 -175.0732>}
 ]
 spine_04 [
  {shapeType=Capsule, length=9.0076, radius=10.3916, center=<5.6234 -0.6779 0>}
  {shapeType=Capsule, length=10.0312, radius=9.2812, center=<-1.2326 1.308 0>, rotation=<rot 90 0 0>}
 ]
 spine_05 [{shapeType=Capsule, length=9.678, radius=12.3571, center=<0.6969 1.91 0>}]
 thigh_l []
 thigh_r []
 upperarm_l []
 upperarm_r []
}
constraints []
