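# Physics asset for the F_MED_PandaSpeak female player body cosmetic, set up for dynamic motion of its apron.
# Five bone chains (front center, left, right and both sides) hang from the pelvis, each running from a base bone through five dyn bones.
# Upper joints rotate freely on tiny capsules, while the lower three lock twist, set tighter swing limits and carry larger offset capsules.
# Unconstrained thigh and calf capsules, plus the pelvis, provide body collision shapes.
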PhysicsAsset /BRCosmetics/Characters/Player/Female/Medium/Bodies/F_MED_PandaSpeak/Meshes/F_MED_PandaSpeak_Physics
bodies {
 base_apron_fr_c [{shapeType=Capsule, length=1, radius=1, shapeName=base_apron_fr_c_capsule}]
 base_apron_fr_l [{shapeType=Capsule, length=1, radius=1, shapeName=base_apron_fr_l_capsule}]
 base_apron_fr_r [{shapeType=Capsule, length=1, radius=1, shapeName=base_apron_fr_r_capsule}]
 base_apron_side_l [{shapeType=Capsule, length=1, radius=1, shapeName=base_apron_side_l_capsule}]
 base_apron_side_r [{shapeType=Capsule, length=1, radius=1, shapeName=base_apron_side_r_capsule}]
 calf_l [{
   shapeType=Capsule
   length=29.0602
   radius=5.6867
   shapeName=calf_l_capsule
   center=<-23.8857 -1.3331 -0.0404>
   rotation=<rot 0.1488 90.638 -87.4199>
  }]
 calf_r [{
   shapeType=Capsule
   length=29.0602
   radius=5.6867
   shapeName=calf_l_capsule
   center=<23.8857 1.3331 0.0404>
   rotation=<rot 0.1488 90.638 -87.4199>
  }]
 dyn_apron_fr_c_00 [{shapeType=Capsule, length=1, radius=1, shapeName=dyn_apron_fr_c_00_capsule}]
 dyn_apron_fr_c_01 [{shapeType=Capsule, length=1, radius=1, shapeName=dyn_apron_fr_c_01_capsule}]
 dyn_apron_fr_c_02 [{shapeType=Capsule, length=8, radius=5, shapeName=dyn_apron_fr_c_02_capsule, center=<2 -3 0>}]
 dyn_apron_fr_c_03 [{shapeType=Capsule, length=8, radius=5, shapeName=dyn_apron_fr_c_03_capsule, center=<2 -2 0>}]
 dyn_apron_fr_c_04 [{shapeType=Capsule, length=8, radius=5, shapeName=dyn_apron_fr_c_04_capsule, center=<2 -2 0>}]
 dyn_apron_fr_l_00 [{shapeType=Capsule, length=1, radius=1, shapeName=dyn_apron_fr_l_00_capsule}]
 dyn_apron_fr_l_01 [{shapeType=Capsule, length=1, radius=1, shapeName=dyn_apron_fr_l_01_capsule}]
 dyn_apron_fr_l_02 [{shapeType=Capsule, length=8, radius=5, shapeName=dyn_apron_fr_l_02_capsule, center=<2 -3 0>}]
 dyn_apron_fr_l_03 [{shapeType=Capsule, length=8, radius=5, shapeName=dyn_apron_fr_l_03_capsule, center=<2 -2 0>}]
 dyn_apron_fr_l_04 [{shapeType=Capsule, length=8, radius=5, shapeName=dyn_apron_fr_l_04_capsule, center=<2 -2 0>}]
 dyn_apron_fr_r_00 [{shapeType=Capsule, length=1, radius=1, shapeName=dyn_apron_fr_r_00_capsule}]
 dyn_apron_fr_r_01 [{shapeType=Capsule, length=1, radius=1, shapeName=dyn_apron_fr_r_01_capsule}]
 dyn_apron_fr_r_02 [{shapeType=Capsule, length=8, radius=5, shapeName=dyn_apron_fr_r_02_capsule, center=<-2 3 0>}]
 dyn_apron_fr_r_03 [{shapeType=Capsule, length=8, radius=5, shapeName=dyn_apron_fr_r_03_capsule, center=<-2 2 0>}]
 dyn_apron_fr_r_04 [{shapeType=Capsule, length=8, radius=5, shapeName=dyn_apron_fr_r_04_capsule, center=<-2 2 0>}]
 dyn_apron_side_l_00 [{shapeType=Capsule, length=1, radius=1, shapeName=dyn_apron_side_l_00_capsule}]
 dyn_apron_side_l_01 [{shapeType=Capsule, length=1, radius=1, shapeName=dyn_apron_side_l_01_capsule}]
 dyn_apron_side_l_02 [{shapeType=Capsule, length=8, radius=5, shapeName=dyn_apron_side_l_02_capsule, center=<2 -3.5 0>}]
 dyn_apron_side_l_03 [{shapeType=Capsule, length=8, radius=5, shapeName=dyn_apron_side_l_03_capsule, center=<2 -3.5 0>}]
 dyn_apron_side_l_04 [{shapeType=Capsule, length=8, radius=5, shapeName=dyn_apron_side_l_04_capsule, center=<2 -3.5 0>}]
 dyn_apron_side_r_00 [{shapeType=Capsule, length=1, radius=1, shapeName=dyn_apron_side_r_00_capsule}]
 dyn_apron_side_r_01 [{shapeType=Capsule, length=1, radius=1, shapeName=dyn_apron_side_r_01_capsule}]
 dyn_apron_side_r_02 [{shapeType=Capsule, length=8, radius=5, shapeName=dyn_apron_side_r_02_capsule, center=<-2 3.5 0>}]
 dyn_apron_side_r_03 [{shapeType=Capsule, length=8, radius=5, shapeName=dyn_apron_side_r_03_capsule, center=<-2 3.5 0>}]
 dyn_apron_side_r_04 [{shapeType=Capsule, length=8, radius=5, shapeName=dyn_apron_side_r_04_capsule, center=<-2 3.5 0>}]
 pelvis [{
   shapeType=Capsule
   length=5.6945
   radius=10.529
   shapeName=pelvis_capsule
   center=<-12.2547 -4.9955 -0.0538>
   rotation=<rot 0 -178.3484 -8.6567>
  }]
 thigh_l [{
   shapeType=Capsule
   length=36.0207
   radius=8.3067
   shapeName=thigh_l_capsule
   center=<-26.2205 -1.2365 0.6106>
   rotation=<rot 85.7162 7.3175 1.4439>
  }]
 thigh_r [{
   shapeType=Capsule
   length=36.0207
   radius=8.3067
   shapeName=thigh_l_capsule
   center=<26.2205 1.2365 -0.6106>
   rotation=<rot 85.7162 7.3175 -0.2726>
  }]
}
constraints [
 {
  bone1Name=base_apron_fr_c
  bone2Name=pelvis
  swing1LimitDegrees=45
  swing1Motion=Free
  swing2LimitDegrees=45
  swing2Motion=Free
  twistLimitDegrees=45
  twistMotion=Free
 }
 {
  bone1Name=dyn_apron_fr_c_00
  bone2Name=base_apron_fr_c
  swing1LimitDegrees=45
  swing1Motion=Free
  swing2LimitDegrees=45
  swing2Motion=Free
  twistLimitDegrees=45
  twistMotion=Free
 }
 {
  bone1Name=dyn_apron_fr_c_01
  bone2Name=dyn_apron_fr_c_00
  swing1LimitDegrees=45
  swing1Motion=Free
  swing2LimitDegrees=45
  swing2Motion=Free
  twistLimitDegrees=45
  twistMotion=Free
 }
 {
  bone1Name=dyn_apron_fr_c_02
  bone2Name=dyn_apron_fr_c_01
  swing1LimitDegrees=20
  swing2LimitDegrees=10
  twistLimitDegrees=45
  twistMotion=Locked
 }
 {
  bone1Name=dyn_apron_fr_c_03
  bone2Name=dyn_apron_fr_c_02
  swing1LimitDegrees=20
  swing2LimitDegrees=10
  twistLimitDegrees=45
  twistMotion=Locked
 }
 {
  bone1Name=dyn_apron_fr_c_04
  bone2Name=dyn_apron_fr_c_03
  swing1LimitDegrees=20
  swing2LimitDegrees=10
  twistLimitDegrees=45
  twistMotion=Locked
 }
 {
  bone1Name=base_apron_fr_l
  bone2Name=pelvis
  swing1LimitDegrees=45
  swing1Motion=Free
  swing2LimitDegrees=45
  swing2Motion=Free
  twistLimitDegrees=45
  twistMotion=Free
 }
 {
  bone1Name=dyn_apron_fr_l_00
  bone2Name=base_apron_fr_l
  swing1LimitDegrees=45
  swing1Motion=Free
  swing2LimitDegrees=45
  swing2Motion=Free
  twistLimitDegrees=45
  twistMotion=Free
 }
 {
  bone1Name=dyn_apron_fr_l_01
  bone2Name=dyn_apron_fr_l_00
  swing1LimitDegrees=45
  swing1Motion=Free
  swing2LimitDegrees=45
  swing2Motion=Free
  twistLimitDegrees=45
  twistMotion=Free
 }
 {
  bone1Name=dyn_apron_fr_l_02
  bone2Name=dyn_apron_fr_l_01
  swing1LimitDegrees=20
  swing2LimitDegrees=10
  twistLimitDegrees=45
  twistMotion=Locked
 }
 {
  bone1Name=dyn_apron_fr_l_03
  bone2Name=dyn_apron_fr_l_02
  swing1LimitDegrees=20
  swing2LimitDegrees=10
  twistLimitDegrees=45
  twistMotion=Locked
 }
 {
  bone1Name=dyn_apron_fr_l_04
  bone2Name=dyn_apron_fr_l_03
  swing1LimitDegrees=20
  swing2LimitDegrees=10
  twistLimitDegrees=45
  twistMotion=Locked
 }
 {
  bone1Name=base_apron_fr_r
  bone2Name=pelvis
  swing1LimitDegrees=45
  swing1Motion=Free
  swing2LimitDegrees=45
  swing2Motion=Free
  twistLimitDegrees=45
  twistMotion=Free
 }
 {
  bone1Name=dyn_apron_fr_r_00
  bone2Name=base_apron_fr_r
  swing1LimitDegrees=45
  swing1Motion=Free
  swing2LimitDegrees=45
  swing2Motion=Free
  twistLimitDegrees=45
  twistMotion=Free
 }
 {
  bone1Name=dyn_apron_fr_r_01
  bone2Name=dyn_apron_fr_r_00
  swing1LimitDegrees=45
  swing1Motion=Free
  swing2LimitDegrees=45
  swing2Motion=Free
  twistLimitDegrees=45
  twistMotion=Free
 }
 {
  bone1Name=dyn_apron_fr_r_02
  bone2Name=dyn_apron_fr_r_01
  swing1LimitDegrees=20
  swing2LimitDegrees=10
  twistLimitDegrees=45
  twistMotion=Locked
 }
 {
  bone1Name=dyn_apron_fr_r_03
  bone2Name=dyn_apron_fr_r_02
  swing1LimitDegrees=20
  swing2LimitDegrees=10
  twistLimitDegrees=45
  twistMotion=Locked
 }
 {
  bone1Name=dyn_apron_fr_r_04
  bone2Name=dyn_apron_fr_r_03
  swing1LimitDegrees=20
  swing2LimitDegrees=10
  twistLimitDegrees=45
  twistMotion=Locked
 }
 {
  bone1Name=base_apron_side_l
  bone2Name=pelvis
  swing1LimitDegrees=45
  swing1Motion=Free
  swing2LimitDegrees=45
  swing2Motion=Free
  twistLimitDegrees=45
  twistMotion=Free
 }
 {
  bone1Name=dyn_apron_side_l_00
  bone2Name=base_apron_side_l
  swing1LimitDegrees=45
  swing1Motion=Free
  swing2LimitDegrees=45
  swing2Motion=Free
  twistLimitDegrees=45
  twistMotion=Free
 }
 {
  bone1Name=dyn_apron_side_l_01
  bone2Name=dyn_apron_side_l_00
  swing1LimitDegrees=45
  swing1Motion=Free
  swing2LimitDegrees=45
  swing2Motion=Free
  twistLimitDegrees=45
  twistMotion=Free
 }
 {
  bone1Name=dyn_apron_side_l_02
  bone2Name=dyn_apron_side_l_01
  swing1LimitDegrees=20
  swing2LimitDegrees=10
  twistLimitDegrees=45
  twistMotion=Locked
 }
 {
  bone1Name=dyn_apron_side_l_03
  bone2Name=dyn_apron_side_l_02
  swing1LimitDegrees=20
  swing2LimitDegrees=10
  twistLimitDegrees=45
  twistMotion=Locked
 }
 {
  bone1Name=dyn_apron_side_l_04
  bone2Name=dyn_apron_side_l_03
  swing1LimitDegrees=20
  swing2LimitDegrees=10
  twistLimitDegrees=45
  twistMotion=Locked
 }
 {
  bone1Name=base_apron_side_r
  bone2Name=pelvis
  swing1LimitDegrees=45
  swing1Motion=Free
  swing2LimitDegrees=45
  swing2Motion=Free
  twistLimitDegrees=45
  twistMotion=Free
 }
 {
  bone1Name=dyn_apron_side_r_00
  bone2Name=base_apron_side_r
  swing1LimitDegrees=45
  swing1Motion=Free
  swing2LimitDegrees=45
  swing2Motion=Free
  twistLimitDegrees=45
  twistMotion=Free
 }
 {
  bone1Name=dyn_apron_side_r_01
  bone2Name=dyn_apron_side_r_00
  swing1LimitDegrees=45
  swing1Motion=Free
  swing2LimitDegrees=45
  swing2Motion=Free
  twistLimitDegrees=45
  twistMotion=Free
 }
 {
  bone1Name=dyn_apron_side_r_02
  bone2Name=dyn_apron_side_r_01
  swing1LimitDegrees=20
  swing2LimitDegrees=10
  twistLimitDegrees=45
  twistMotion=Locked
 }
 {
  bone1Name=dyn_apron_side_r_03
  bone2Name=dyn_apron_side_r_02
  swing1LimitDegrees=20
  swing2LimitDegrees=10
  twistLimitDegrees=45
  twistMotion=Locked
 }
 {
  bone1Name=dyn_apron_side_r_04
  bone2Name=dyn_apron_side_r_03
  swing1LimitDegrees=20
  swing2LimitDegrees=10
  twistLimitDegrees=45
  twistMotion=Locked
 }
]
