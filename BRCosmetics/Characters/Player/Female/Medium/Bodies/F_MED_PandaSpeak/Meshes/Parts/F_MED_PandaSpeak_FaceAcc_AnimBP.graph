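# Post-process style AnimBlueprint for the PandaSpeak female face accessory, which follows the attached parent mesh and adds hair motion.
# The copied pose runs through a Control Rig, then four chained AnimDynamics nodes spring the left and right hair bones.
# Each update feeds the rig assault-rifle and PandaSpeak emote montage states.
# It also sets spring and scale variables from targeting, but the dynamics nodes show fixed values instead.
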AnimBlueprint /BRCosmetics/Characters/Player/Female/Medium/Bodies/F_MED_PandaSpeak/Meshes/Parts/F_MED_PandaSpeak_FaceAcc_AnimBP
target_skeleton=&/F_MED_PandaSpeak_FaceAcc_Skeleton
graphs [
 AnimGraph {nodes [
   n1 = Control Rig!(
    Source=Copy Pose From Mesh() {bUseAttachedParent=true}
    Alpha
    LargeGunOut:double=Get large_gun_out() -> (large_gun_out:double)
    PandaSpeakEmote:bool=Get PandaSpeakEmote() -> (PandaSpeakEmote:bool)
   ) -> (Pose) {controlRigAssetReference {blueprintRigClass=&/F_MED_PandaSpeak_FaceAcc_CtrlRig.F_MED_PandaSpeak_FaceAcc_CtrlRig_C}}
   n2 = Anim Dynamics - Bone: dyn_hair_l_01!(
    ComponentPose=Local To Component(LocalPose=@n1)
    Alpha=1
    AngularSpringConstant:float=8000
    ComponentLinearAccScale:Vector="0.5,0.5,0.5"
    ComponentLinearVelScale:Vector="0.5,0.5,0.5"
    ExternalForce="0,0,0"
   ) -> (Pose) {
    angularDampingOverride=1
    angularSpringConstant=8000
    bAngularSpring=true
    bOverrideAngularDamping=true
    bUseGravityOverride=true
    gravityScale=0.35
    simulationSpace=RootRelative
    boundBone {boneName=dyn_hair_l_01}
    componentAppliedLinearAccClamp=<3000 3000 3000>
    componentLinearAccScale=<0.5 0.5 0.5>
    componentLinearVelScale=<0.5 0.5 0.5>
    gravityOverride {z=-980}
    physicsBodyDefinitions [{
      boundBone {boneName=dyn_hair_l_01}
      boxExtents=<20 30 30>
      constraintSetup {angularLimitsMax=<2 11 0>, angularLimitsMin=<-2 -2 -11>, angularTarget=<1 0 0>}
      localJointOffset=<3 0 0>
     }]
    simSpaceSettings {maxAngularAcceleration=30, maxAngularVelocity=5, simSpaceAngularAlpha=1}
   }
   n3 = Anim Dynamics - Bone: dyn_hair_l_02!(
    ComponentPose=@n2
    Alpha=1
    AngularSpringConstant:float=8000
    ComponentLinearAccScale:Vector="0.5,0.5,0.5"
    ComponentLinearVelScale:Vector="0.5,0.5,0.5"
    ExternalForce="0,0,0"
   ) -> (Pose) {
    angularDampingOverride=1
    angularSpringConstant=8000
    bAngularSpring=true
    bOverrideAngularDamping=true
    bUseGravityOverride=true
    gravityScale=0.35
    simulationSpace=RootRelative
    boundBone {boneName=dyn_hair_l_02}
    componentAppliedLinearAccClamp=<3000 3000 3000>
    componentLinearAccScale=<0.5 0.5 0.5>
    componentLinearVelScale=<0.5 0.5 0.5>
    gravityOverride {z=-980}
    physicsBodyDefinitions [{
      boundBone {boneName=dyn_hair_l_02}
      boxExtents=<20 30 30>
      constraintSetup {angularLimitsMax=<2 11 0>, angularLimitsMin=<-2 -5 -11>, angularTarget=<1 0 0>}
      localJointOffset=<3 0 0>
     }]
    simSpaceSettings {maxAngularAcceleration=30, maxAngularVelocity=5, simSpaceAngularAlpha=1}
   }
   n4 = Anim Dynamics - Bone: dyn_hair_r_01!(
    ComponentPose=@n3
    Alpha=1
    AngularSpringConstant:float=8000
    ComponentLinearAccScale:Vector="0.5,0.5,0.5"
    ComponentLinearVelScale:Vector="0.5,0.5,0.5"
    ExternalForce="0,0,0"
   ) -> (Pose) {
    angularDampingOverride=1
    angularSpringConstant=8000
    bAngularSpring=true
    bOverrideAngularDamping=true
    bUseGravityOverride=true
    gravityScale=0.35
    simulationSpace=RootRelative
    boundBone {boneName=dyn_hair_r_01}
    componentAppliedLinearAccClamp=<3000 3000 3000>
    componentLinearAccScale=<0.5 0.5 0.5>
    componentLinearVelScale=<0.5 0.5 0.5>
    gravityOverride {z=-980}
    physicsBodyDefinitions [{
      boundBone {boneName=dyn_hair_r_01}
      boxExtents=<20 30 30>
      constraintSetup {angularLimitsMax=<2 11 0>, angularLimitsMin=<-2 -2 -11>, angularTarget=<1 0 0>}
      localJointOffset=<3 0 0>
     }]
    simSpaceSettings {maxAngularAcceleration=30, maxAngularVelocity=5, simSpaceAngularAlpha=1}
   }
   n5 = Anim Dynamics - Bone: dyn_hair_r_02!(
    ComponentPose=@n4
    Alpha=1
    AngularSpringConstant:float=8000
    ComponentLinearAccScale:Vector="0.5,0.5,0.5"
    ComponentLinearVelScale:Vector="0.5,0.5,0.5"
    ExternalForce="0,0,0"
   ) -> (Pose) {
    angularDampingOverride=1
    angularSpringConstant=8000
    bAngularSpring=true
    bOverrideAngularDamping=true
    bUseGravityOverride=true
    gravityScale=0.35
    simulationSpace=RootRelative
    boundBone {boneName=dyn_hair_r_02}
    componentAppliedLinearAccClamp=<3000 3000 3000>
    componentLinearAccScale=<0.5 0.5 0.5>
    componentLinearVelScale=<0.5 0.5 0.5>
    gravityOverride {z=-980}
    physicsBodyDefinitions [{
      boundBone {boneName=dyn_hair_r_02}
      boxExtents=<20 30 30>
      constraintSetup {angularLimitsMax=<2 11 0>, angularLimitsMin=<-2 -5 -11>, angularTarget=<1 0 0>}
      localJointOffset=<3 0 0>
     }]
    simSpaceSettings {maxAngularAcceleration=30, maxAngularVelocity=5, simSpaceAngularAlpha=1}
   }
   Output Pose(Result=Component To Local(ComponentPose=@n5)) {layerGroup=None}
  ]}
 EventGraph {nodes [
   Event BlueprintUpdateAnimation()
   > n1 = Sequence!(execute) -> (then_0, then_1, then_2:Exec, then_3:Exec)
   > n2 = Cast To FortPlayerAnimInstance(Object=GetAnimInstance(self=GetOwnerSkeletalMeshComponent()))
   > Set rifle_out(rifle_out:bool=false) -> (Output_Get:bool)
   > n3 = Switch on EFortWeaponCoreAnimation!(execute, Selection:EFortWeaponCoreAnimation Enum=Get WeaponCoreAnim(self=@n2)) -> (
    Melee:Exec
    Pistol:Exec
    Shotgun:Exec
    PaperBlueprint:Exec
    Rifle:Exec
    MeleeOneHand:Exec
    MachinePistol:Exec
    RocketLauncher:Exec
    GrenadeLauncher:Exec
    GoingCommando:Exec
    AssaultRifle:Exec
    TacticalShotgun:Exec
    SniperRifle:Exec
    TrapPlacement:Exec
    ShoulderLauncher:Exec
    AbilityDecoTool:Exec
    Crossbow:Exec
    C4:Exec
    RemoteControl:Exec
    DualWield:Exec
    AR_BullPup:Exec
    AR_ForwardGrip:Exec
    MedPackPaddles:Exec
    SMG_P90:Exec
    AR_DrumGun:Exec
    Consumable_Small:Exec
    Consumable_Large:Exec
    Balloon:Exec
    MountedTurret:Exec
    CreativeTool:Exec
    ExplosiveBow:Exec
    AshtonIndigo:Exec
    AshtonChicago:Exec
    MeleeDualWield:Exec
    Unarmed:Exec
   )
   Branch(execute=@n3.AssaultRifle)
   > Set rifle_out(rifle_out=true)
   Set large_gun_out(execute=@n1.then_1, large_gun_out:double="To Float (Boolean)"(InBool=Get rifle_out() -> (rifle_out:bool))) -> (Output_Get:double)
   n4 = Branch(execute=@n1.then_2, Condition=Get bIsTargeting())
   > Set LinearScale(LinearScale:Vector="0.2,0.2,0.2") -> (Output_Get:Vector)
   > Set Angular Spring(Angular Spring:double=2000) -> (Output_Get:double)
   Set LinearScale(execute=@n4.else, LinearScale="0.5,0.5,0.5")
   > Set Angular Spring(Angular Spring=8000)
   n5 = GetAnimInstance(self=GetOwnerSkeletalMeshComponent())
   Is Valid(exec=@n1.then_3, InputObject=@n5)
   > Set PandaSpeakEmote(PandaSpeakEmote:bool=Montage_IsPlaying(self=@n5, Montage=/BRCosmetics/Animation/Game/MainPlayer/Emotes/PandaSpeak/Hero/Emote_PandaSpeak_Hero_M)) -> (Output_Get:bool)
   Reroute Node(InputPin:Wildcard) -> (OutputPin:Wildcard)
  ]}
]
variables [
 Angular Limits Max {type=Vector}
 Angular Limits Min {type=Vector}
 Angular Spring {type=number}
 large_gun_out {type=number}
 LinearScale {type=Vector}
 PandaSpeakEmote {type=boolean}
 rifle_out {type=boolean}
]
